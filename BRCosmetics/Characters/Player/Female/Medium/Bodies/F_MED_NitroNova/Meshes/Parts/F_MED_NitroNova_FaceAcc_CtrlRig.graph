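# Control Rig for NitroNova's face accessory, posing the left, center and right dynamic hair chains against body motion.
# Weighted parent constraints tie hair nulls to head, neck, spine and clavicles.
# Swing/twist and spherical pose readers drive corrective offsets, modulated by interpolated states like crouch-walking.
# Twist distribution precedes stretchy B-spline IK, which covers longer chains when rigid bodies are off or backbling is worn, otherwise only the upper five bones.
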ControlRigBlueprint /BRCosmetics/Characters/Player/Female/Medium/Bodies/F_MED_NitroNova/Meshes/Parts/F_MED_NitroNova_FaceAcc_CtrlRig
graphs [
 AlterTransform_ContainedGraph {nodes [
   n1 = RigVMFunctionEntryNode()
   > For_Each = RigVMDispatchNode(Array=@n1.transforms) {template="DISPATCH_RigVMDispatch_ArrayIterator(in Array,out Element,out Index,out Count,out Ratio)"}
   > &RigUnit_OffsetTransformForItem(Item=@For_Each.Element.Item, OffsetTransform=@For_Each.Element.Transform, Weight=@n1.weights, bPropagateToChildren=@n1.inheritsTransform)
   RigVMFunctionReturnNode(ExecuteContext=@For_Each.Completed)
  ]}
 CollapseNode_1_ContainedGraph {nodes [
   RigVMFunctionEntryNode()
   > DeltaRotation v4_8 = RigVMFunctionReferenceNode(
    Rotation Order=ZYX
    child=(Type=Bone,Name="head")
    onlyTwist=false
    outputMinMax=(X=0,Y=1)
    parent=(Type=Null,Name="head_null")
    premult=1
    rotationMinMax=(X=0,Y=-50)
    triggerAxis=Z
    twistAxis=(X=1,Y=0,Z=0)
   )
   > &RigUnit_OffsetTransformForItem(
    Item=(Type=Bone,Name="dyn_hair_0_l")
    OffsetTransform=(Rotation=(X=0,Y=0,Z=0,W=1),Translation=(X=-3,Y=0,Z=2),Scale3D=(X=1,Y=1,Z=1))
    bPropagateToChildren=true
    Weight=@"DeltaRotation v4_8".outputWeight
   )
   AlphaInterp = &RigVMFunction_AlphaInterp(
    InterpSpeedDecreasing=8
    InterpSpeedIncreasing=10
    Scale=1
    bClampResult=False
    Value=&RigVMFunction_MathBoolToFloat(Value=RigVMVariableNode() {variable=isCrouching}.Value).Result
   )
   Multiply = &RigVMFunction_MathDoubleMul(B=@AlphaInterp.Result, A=@"DeltaRotation v4_8".outputWeight)
   > &RigUnit_ModifyTransforms(
    ItemToModify="((Item=(Type=Bone,Name=\"dyn_hair_2_l\"),Transform=(Rotation=(X=0,Y=0,Z=0,W=-1),Translation=(X=0,Y=0,Z=2),Scale3D=(X=1,Y=1,Z=1))),(Item=(Type=Bone,Name=\"dyn_hair_0_l\"),Transform=(Rotation=(X=0,Y=0,Z=0,W=-1),Translation=(X=2,Y=0,Z=0),Scale3D=(X=1,Y=1,Z=1))),( ...(truncated)"
    WeightMinimum=0
    Weight=@Multiply.Result
   )
   > DeltaRotation v4_14 = RigVMFunctionReferenceNode(
    Rotation Order=ZYX
    child=(Type=Bone,Name="head")
    debugRotation=false
    outputMinMax=(X=0,Y=1)
    parent=(Type=Null,Name="head_null")
    premult=1
    rotationMinMax=(X=-50,Y=-89)
    triggerAxis=Z
    twistAxis=(X=1,Y=0,Z=0)
   )
   n1 = &RigVMFunction_MathDoubleMul(A=@"DeltaRotation v4_14".outputWeight, B=RigVMVariableNode() {variable=isCrouchWalking}.Value)
   > &RigUnit_ModifyTransforms(
    ItemToModify="((Item=(Type=Bone,Name=\"dyn_hair_1_c\"),Transform=(Rotation=(X=0,Y=0.1736,Z=0,W=0.9848),Translation=(X=2,Y=0,Z=2),Scale3D=(X=1,Y=1,Z=1))),(Item=(Type=Bone,Name=\"dyn_hair_2_l\"),Transform=(Rotation=(X=0,Y=0,Z=0,W=-1),Translation=(X=0,Y=0,Z=0),Scale3D=(X=1,Y=1 ...(truncated)"
    WeightMinimum=0
    Weight=@n1.Result
   )
   > &RigUnit_ModifyTransforms(
    ItemToModify=((Item=(Type=Bone,Name="dyn_hairShoulder_l"),Transform=(Rotation=(X=0,Y=0,Z=0,W=1),Translation=(X=-4,Y=0,Z=-1),Scale3D=(X=1,Y=1,Z=1))))
    WeightMinimum=0
    Weight=@n1.Result
   )
   > &RigUnit_OffsetTransformForItem(
    Item=(Type=Bone,Name="dyn_hair_0_c")
    OffsetTransform=(Rotation=(X=0,Y=0,Z=0,W=1),Translation=(X=-2,Y=0,Z=0),Scale3D=(X=1,Y=1,Z=1))
    bPropagateToChildren=True
    Weight=@n1.Result
   )
   > DeltaRotation v4_2 = RigVMFunctionReferenceNode(
    Rotation Order=ZYX
    child=(Type=Bone,Name="upperarm_l")
    debugRotation=false
    onlyTwist=false
    outputMinMax=(X=0,Y=1)
    parent=(Type=Null,Name="upperarm_l_null")
    premult=1
    rotationMinMax=(X=0,Y=-45)
    triggerAxis=Y
    twistAxis=(X=1,Y=0,Z=0)
   )
   > &RigUnit_ModifyTransforms(
    ItemToModify=((Item=(Type=Bone,Name="dyn_hair_2_l"),Transform=(Rotation=(X=0,Y=0,Z=0,W=-1),Translation=(X=0,Y=0,Z=2),Scale3D=(X=1,Y=1,Z=1))))
    WeightMinimum=0
    Weight=@"DeltaRotation v4_2".outputWeight
   )
   > DeltaRotation v4_3 = RigVMFunctionReferenceNode(
    Rotation Order=ZYX
    child=(Type=Bone,Name="upperarm_r")
    debugRotation=false
    onlyTwist=false
    outputMinMax=(X=0,Y=1)
    parent=(Type=Null,Name="upperarm_r_null")
    premult=1
    rotationMinMax=(X=0,Y=-45)
    triggerAxis=Y
    twistAxis=(X=1,Y=0,Z=0)
   )
   > &RigUnit_ModifyTransforms(
    ItemToModify=((Item=(Type=Bone,Name="dyn_hair_2_r"),Transform=(Rotation=(X=0,Y=0,Z=0,W=1),Translation=(X=0,Y=0,Z=4),Scale3D=(X=1,Y=1,Z=1))))
    WeightMinimum=0
    Weight=@"DeltaRotation v4_3".outputWeight
   )
   > RigVMFunctionReturnNode()
   &RigUnit_ModifyTransforms(
    ExecutePin=()
    ItemToModify=((Item=(Type=Bone,Name="dyn_hair_2_l"),Transform=(Rotation=(X=0,Y=0,Z=0,W=1),Translation=(X=0,Y=-2,Z=2),Scale3D=(X=1,Y=1,Z=1))))
    Weight=1
    WeightMinimum=0
   )
   &RigUnit_SphericalPoseReader(
    ActiveRegionScaleFactors=(PositiveWidth=1,NegativeWidth=1,PositiveHeight=1,NegativeHeight=1)
    ActiveRegionSize=0.05
    Debug=(bDrawDebug=true,bDraw2D=False,bDrawLocalAxes=False,DebugScale=25,DebugSegments=20,DebugThickness=0.25)
    DriverAxis=(X=1,Y=0,Z=0)
    DriverItem=(Type=Bone,Name="head")
    ExecutePin=()
    FalloffRegionScaleFactors=(PositiveWidth=1,NegativeWidth=1,PositiveHeight=1,NegativeHeight=1)
    FalloffSize=0.2
    FlipHeightScaling=False
    FlipWidthScaling=False
    OptionalParentItem=(Type=Bone,Name="spine_05")
    RotationOffset=(X=0,Y=-62,Z=-89)
   )
   &RigUnit_SphericalPoseReader(
    ActiveRegionScaleFactors=(PositiveWidth=1,NegativeWidth=1,PositiveHeight=1,NegativeHeight=1)
    ActiveRegionSize=0.05
    Debug=(bDrawDebug=true,bDraw2D=False,bDrawLocalAxes=False,DebugScale=25,DebugSegments=20,DebugThickness=0.25)
    DriverAxis=(X=1,Y=0,Z=0)
    DriverItem=(Type=Bone,Name="head")
    ExecutePin=()
    FalloffRegionScaleFactors=(PositiveWidth=1,NegativeWidth=1,PositiveHeight=1,NegativeHeight=1)
    FalloffSize=0.2
    FlipHeightScaling=False
    FlipWidthScaling=False
    OptionalParentItem=(Type=Bone,Name="spine_05")
    RotationOffset=(X=0,Y=-70,Z=-60)
   )
  ]}
 CollapseNode_ContainedGraph {nodes [
   n1 = RigVMFunctionEntryNode()
   constraints = RigVMCollapseNode(ExecutePin=@n1.ExecutePin)
   RigVMFunction_Sequence = RigVMAggregateNode(ExecuteContext=@constraints.ExecutePin)
   n2 = &RigVMFunction_Sequence(ExecuteContext=@RigVMFunction_Sequence.A)
   DeltaRotation v4_1 = RigVMFunctionReferenceNode(
    Rotation Order=ZYX
    child=(Type=Bone,Name="head")
    onlyTwist=false
    outputMinMax=(X=0,Y=1)
    parent=(Type=Null,Name="head_null")
    premult=1
    rotationMinMax=(X=0,Y=-50)
    triggerAxis=Z
    twistAxis=(X=1,Y=0,Z=0)
    ExecuteContext=@n2.A
   )
   n3 = &RigVMFunction_MathDoubleMul(
    A=@"DeltaRotation v4_1".outputWeight
    B=&RigVMFunction_MathDoubleRemap(SourceMaximum=1, SourceMinimum=0, TargetMaximum=0, TargetMinimum=1, Value=RigVMVariableNode() {variable=isCrouchWalking}.Value).Result
   )
   > &RigUnit_ModifyTransforms(
    ItemToModify="((Item=(Type=Bone,Name=\"dyn_hairShoulder_l\"),Transform=(Rotation=(X=0,Y=0,Z=0,W=1),Translation=(X=-1.85,Y=0,Z=0),Scale3D=(X=1,Y=1,Z=1))),(Item=(Type=Bone,Name=\"dyn_hairShoulder_r\"),Transform=(Rotation=(X=0,Y=0,Z=0,W=-1),Translation=(X=-3.7777,Y=-0.0001,Z=0 ...(truncated)"
    WeightMinimum=0
    Weight=@n3.Result
   )
   > &RigUnit_ModifyTransforms(
    ItemToModify="((Item=(Type=Null,Name=\"dyn_hair_3_l_null\"),Transform=(Rotation=(X=0,Y=0,Z=0,W=1),Translation=(X=0,Y=-2,Z=0),Scale3D=(X=1,Y=1,Z=1))),(Item=(Type=Null,Name=\"dyn_hair_3_r_null\"),Transform=(Rotation=(X=0,Y=0,Z=0,W=-1),Translation=(X=0,Y=-2,Z=5),Scale3D=(X=1,Y ...(truncated)"
    WeightMinimum=0
    Weight=@n3.Result
   )
   > DeltaRotation v4_4 = RigVMFunctionReferenceNode(
    Rotation Order=ZYX
    absoluteValue=true
    child=(Type=Bone,Name="head")
    debugRotation=false
    onlyTwist=true
    outputMinMax=(X=0,Y=1)
    parent=(Type=Null,Name="head_null")
    premult=1
    rotationMinMax=(X=0,Y=-60)
    twistAxis=(X=1,Y=0,Z=0)
   )
   n4 = RigVMVariableNode() {variable=isCrouchWalking}
   > &RigUnit_ModifyTransforms(
    ItemToModify=((Item=(Type=Bone,Name="dyn_hairShoulder_l"),Transform=(Rotation=(X=0,Y=0,Z=0,W=-1),Translation=(X=-4,Y=0,Z=0),Scale3D=(X=1,Y=1,Z=1))))
    WeightMinimum=0
    Weight=&RigVMFunction_MathDoubleMul(
     A=@"DeltaRotation v4_4".outputWeight
     B=&RigVMFunction_MathDoubleRemap(SourceMaximum=1, SourceMinimum=0, TargetMaximum=0, TargetMinimum=1, Value=@n4.Value).Result
    ).Result
   )
   > DeltaRotation v4_5 = RigVMFunctionReferenceNode(
    Rotation Order=ZYX
    absoluteValue=true
    child=(Type=Bone,Name="head")
    debugRotation=false
    onlyTwist=true
    outputMinMax=(X=0,Y=1)
    parent=(Type=Null,Name="head_null")
    premult=1
    rotationMinMax=(X=0,Y=60)
    twistAxis=(X=1,Y=0,Z=0)
   )
   n5 = RigVMVariableNode() {variable=isCrouchWalking}
   > &RigUnit_ModifyTransforms(
    ItemToModify="((Item=(Type=Null,Name=\"dyn_hair_2_r_null\"),Transform=(Rotation=(X=0,Y=0,Z=0,W=1),Translation=(X=4,Y=0,Z=6),Scale3D=(X=1,Y=1,Z=1))),(Item=(Type=Bone,Name=\"dyn_hairShoulder_r\"),Transform=(Rotation=(X=0,Y=0,Z=0,W=1),Translation=(X=0,Y=2,Z=-7.5),Scale3D=(X=1, ...(truncated)"
    WeightMinimum=0
    Weight=&RigVMFunction_MathDoubleMul(
     A=@"DeltaRotation v4_5".outputWeight
     B=&RigVMFunction_MathDoubleRemap(SourceMaximum=1, SourceMinimum=0, TargetMaximum=0, TargetMinimum=1, Value=@n5.Value).Result
    ).Result
   )
   > ClaviclesPosFix = RigVMCollapseNode()
   DeltaRotation v4_11 = RigVMFunctionReferenceNode(
    Rotation Order=ZYX
    child=(Type=Bone,Name="spine_03")
    debugRotation=false
    onlyTwist=true
    outputMinMax=(X=0,Y=1)
    parent=(Type=Bone,Name="spine_05")
    premult=0
    rotationMinMax=(X=0,Y=-35)
    twistAxis=(X=1,Y=0,Z=0)
    ExecuteContext=@n2.B
   )
   > &RigUnit_ModifyTransforms(
    ItemToModify="((Item=(Type=Null,Name=\"dyn_hair_3_r_null\"),Transform=(Rotation=(X=0,Y=0,Z=0,W=1),Translation=(X=0,Y=0,Z=3),Scale3D=(X=1,Y=1,Z=1))),(Item=(Type=Null,Name=\"dyn_hair_4_r_null\"),Transform=(Rotation=(X=0,Y=0,Z=0,W=1),Translation=(X=0,Y=0,Z=3),Scale3D=(X=1,Y=1, ...(truncated)"
    WeightMinimum=0
    Weight=@"DeltaRotation v4_11".outputWeight
   )
   > DeltaRotation v4_10 = RigVMFunctionReferenceNode(
    Rotation Order=ZYX
    child=(Type=Bone,Name="spine_03")
    debugRotation=false
    onlyTwist=true
    outputMinMax=(X=0,Y=1)
    parent=(Type=Bone,Name="spine_05")
    premult=1
    rotationMinMax=(X=0,Y=35)
    twistAxis=(X=1,Y=0,Z=0)
   )
   > &RigUnit_ModifyTransforms(
    ItemToModify="((Item=(Type=Null,Name=\"dyn_hair_3_l_null\"),Transform=(Rotation=(X=0,Y=0,Z=0,W=1),Translation=(X=0,Y=0,Z=3),Scale3D=(X=1,Y=1,Z=1))),(Item=(Type=Null,Name=\"dyn_hair_4_l_null\"),Transform=(Rotation=(X=0,Y=0,Z=0,W=1),Translation=(X=0,Y=0,Z=3),Scale3D=(X=1,Y=1, ...(truncated)"
    WeightMinimum=0
    Weight=@"DeltaRotation v4_10".outputWeight
   )
   > n6 = &RigUnit_SphericalPoseReader(
    ActiveRegionScaleFactors=(PositiveWidth=1.5,NegativeWidth=2,PositiveHeight=1.5,NegativeHeight=1)
    ActiveRegionSize=0.05
    Debug=(bDrawDebug=false,bDraw2D=False,bDrawLocalAxes=False,DebugScale=25,DebugSegments=20,DebugThickness=0.25)
    DriverAxis=(X=1,Y=0,Z=0)
    DriverItem=(Type=Bone,Name="head")
    FalloffRegionScaleFactors=(PositiveWidth=0.7,NegativeWidth=1,PositiveHeight=1,NegativeHeight=1)
    FalloffSize=0.2
    FlipHeightScaling=False
    FlipWidthScaling=False
    OptionalParentItem=(Type=Bone,Name="spine_05")
    RotationOffset=(X=0,Y=-50,Z=-15)
   )
   > &RigUnit_ModifyTransforms(
    ItemToModify="((Item=(Type=Bone,Name=\"dyn_hairShoulder_r\"),Transform=(Rotation=(X=0,Y=0,Z=0,W=1),Translation=(X=-7,Y=0,Z=0),Scale3D=(X=1,Y=1,Z=1))),(Item=(Type=Null,Name=\"dyn_hair_3_r_null\"),Transform=(Rotation=(X=0,Y=0,Z=0,W=1),Translation=(X=0,Y=-3,Z=0),Scale3D=(X=1,Y ...(truncated)"
    WeightMinimum=0
    Weight=@n6.OutputParam
   )
   > n7 = &RigUnit_SphericalPoseReader(
    ActiveRegionScaleFactors=(PositiveWidth=1,NegativeWidth=1,PositiveHeight=1,NegativeHeight=1)
    ActiveRegionSize=0.05
    Debug=(bDrawDebug=false,bDraw2D=False,bDrawLocalAxes=False,DebugScale=25,DebugSegments=20,DebugThickness=0.25)
    DriverAxis=(X=1,Y=0,Z=0)
    DriverItem=(Type=Bone,Name="head")
    FalloffRegionScaleFactors=(PositiveWidth=3,NegativeWidth=1,PositiveHeight=2,NegativeHeight=1)
    FalloffSize=0.12
    FlipHeightScaling=False
    FlipWidthScaling=False
    OptionalParentItem=(Type=Bone,Name="spine_05")
    RotationOffset=(X=0,Y=-80,Z=10)
   )
   Multiply = &RigVMFunction_MathDoubleMul(A=@n7.OutputParam, B=@"DeltaRotation v4_10".outputWeight)
   > &RigUnit_ModifyTransforms(
    ItemToModify="((Item=(Type=Bone,Name=\"dyn_hairShoulder_r\"),Transform=(Rotation=(X=0,Y=0,Z=0,W=1),Translation=(X=-5,Y=0,Z=0),Scale3D=(X=1,Y=1,Z=1))),(Item=(Type=Null,Name=\"dyn_hair_2_r_null\"),Transform=(Rotation=(X=0,Y=0,Z=0,W=-1),Translation=(X=0,Y=0,Z=5),Scale3D=(X=1,Y ...(truncated)"
    WeightMinimum=0
    Weight=@Multiply.Result
   )
   > DeltaRotation v4_2 = RigVMFunctionReferenceNode(
    Rotation Order=ZYX
    absoluteValue=true
    child=(Type=Bone,Name="head")
    debugRotation=false
    onlyTwist=false
    outputMinMax=(X=0,Y=1)
    parent=(Type=Null,Name="head_null")
    premult=1
    rotationMinMax=(X=0,Y=-10)
    triggerAxis=Y
    twistAxis=(X=1,Y=0,Z=0)
   )
   > n8 = &RigUnit_SphericalPoseReader(
    ActiveRegionScaleFactors=(PositiveWidth=1.5,NegativeWidth=1,PositiveHeight=1,NegativeHeight=1)
    ActiveRegionSize=0.1
    Debug=(bDrawDebug=false,bDraw2D=False,bDrawLocalAxes=False,DebugScale=25,DebugSegments=20,DebugThickness=0.25)
    DriverAxis=(X=-1,Y=0,Z=0)
    DriverItem=(Type=Bone,Name="upperarm_r")
    FalloffRegionScaleFactors=(PositiveWidth=2,NegativeWidth=2,PositiveHeight=1,NegativeHeight=1)
    FalloffSize=0.2
    FlipHeightScaling=False
    FlipWidthScaling=False
    OptionalParentItem=(Type=Bone,Name="None")
    RotationOffset=(X=-75,Y=-60,Z=0)
   )
   > &RigUnit_ModifyTransforms(
    ItemToModify="((Item=(Type=Bone,Name=\"dyn_hairShoulder_r\"),Transform=(Rotation=(X=0,Y=0,Z=0,W=-1),Translation=(X=-2,Y=0,Z=0),Scale3D=(X=1,Y=1,Z=1))),(Item=(Type=Null,Name=\"dyn_hair_2_r_null\"),Transform=(Rotation=(X=0,Y=0,Z=0,W=1),Translation=(X=0,Y=0,Z=5),Scale3D=(X=1,Y ...(truncated)"
    WeightMinimum=0
    Weight=&RigVMFunction_MathDoubleMul(B=@"DeltaRotation v4_2".outputWeight, A=@n8.OutputParam).Result
   )
   > DeltaRotation v4_6 = RigVMFunctionReferenceNode(
    Rotation Order=ZYX
    child=(Type=Bone,Name="pelvis")
    debugRotation=false
    onlyTwist=true
    outputMinMax=(X=0,Y=1)
    parent=(Type=Bone,Name="spine_05")
    premult=1
    rotationMinMax=(X=-30,Y=30)
    twistAxis=(X=1,Y=0,Z=0)
   )
   AlphaInterp = &RigVMFunction_AlphaInterp(
    InterpSpeedDecreasing=8
    InterpSpeedIncreasing=10
    Scale=1
    bClampResult=False
    Value=&RigVMFunction_MathBoolToFloat(Value=RigVMVariableNode() {variable=isPA_Swing}.Value).Result
   )
   RigVMFunction_AnimEvalRichCurve = &RigVMFunction_AnimEvalRichCurve(
    Curve=(EditorCurveData=(Keys=((Value=1),(Time=0.4915),(Time=1,Value=1))))
    SourceMinimum=0
    TargetMinimum=0
    Value=@"DeltaRotation v4_6".outputWeight
   )
   > &RigUnit_ModifyTransforms(
    ItemToModify="((Item=(Type=Null,Name=\"dyn_hair_0_l_null\"),Transform=(Rotation=(X=0,Y=0,Z=0,W=1),Translation=(X=0,Y=0,Z=10),Scale3D=(X=1,Y=1,Z=1))),(Item=(Type=Null,Name=\"dyn_hair_0_r_null\"),Transform=(Rotation=(X=0,Y=0,Z=0,W=1),Translation=(X=0,Y=0,Z=10),Scale3D=(X=1,Y= ...(truncated)"
    WeightMinimum=0
    Weight=&RigVMFunction_MathDoubleMul(B=@AlphaInterp.Result, A=@RigVMFunction_AnimEvalRichCurve.Result).Result
   )
   ItemArray = RigVMDispatchNode(Value=((Type=Bone,Name="dyn_hair_3_c"),(Type=Bone,Name="dyn_hair_4_c"),(Type=Bone,Name="dyn_hair_5_c"),(Type=Bone,Name="dyn_hair_6_c"),(Type=Bone,Name="dyn_hair_7_c"))) {template="DISPATCH_RigVMDispatch_Constant(io Value)"}
   TwistBones = RigVMFunctionReferenceNode(
    Axis to Twist=(bX=True,bY=False,bZ=False)
    Debug=false
    End=(Type=Bone,Name="neck_01")
    Start=(Type=Bone,Name="spine_03")
    Bones to Twist=@ItemArray.Value
    ExecuteContext=@RigVMFunction_Sequence.C
   )
   ItemArray_9 = RigVMDispatchNode(Value=((Type=Bone,Name="dyn_hair_3_l"),(Type=Bone,Name="dyn_hair_4_l"),(Type=Bone,Name="dyn_hair_5_l"),(Type=Bone,Name="dyn_hair_6_l"),(Type=Bone,Name="dyn_hair_7_l"))) {template="DISPATCH_RigVMDispatch_Constant(io Value)"}
   > TwistBones_2 = RigVMFunctionReferenceNode(
    Axis to Twist=(bX=True,bY=False,bZ=False)
    End=(Type=Bone,Name="neck_01")
    Start=(Type=Bone,Name="spine_04")
    Bones to Twist=@ItemArray_9.Value
   )
   ItemArray_1 = RigVMDispatchNode(Value=((Type=Bone,Name="dyn_hair_3_r"),(Type=Bone,Name="dyn_hair_4_r"),(Type=Bone,Name="dyn_hair_5_r"),(Type=Bone,Name="dyn_hair_6_r"),(Type=Bone,Name="dyn_hair_7_r"))) {template="DISPATCH_RigVMDispatch_Constant(io Value)"}
   > TwistBones_1 = RigVMFunctionReferenceNode(
    Axis to Twist=(bX=True,bY=False,bZ=False)
    End=(Type=Bone,Name="spine_05")
    Start=(Type=Bone,Name="spine_04")
    Bones to Twist=@ItemArray_1.Value
   )
   RigVMFunction_MathBoolNot = &RigVMFunction_MathBoolNot(Value=RigVMVariableNode() {variable=PubIsRigidBodiesEnabled}.Value)
   RigVMFunction_MathBoolOr = &RigVMFunction_MathBoolOr(A=@RigVMFunction_MathBoolNot.Result, B=RigVMVariableNode() {variable=PubIsWearingBackbling}.Value)
   > SplineIK = RigVMCollapseNode(Condition=@RigVMFunction_MathBoolOr.Result)
   > ExtremeHead_Fix = RigVMCollapseNode()
   RigVMFunctionReturnNode(ExecutePin=@RigVMFunction_Sequence.E)
   &RigVMFunction_MathDoubleMul(
    A=@Multiply.Result
    B=&RigVMFunction_MathDoubleRemap(SourceMaximum=1, SourceMinimum=0, TargetMaximum=1, TargetMinimum=0, Value=RigVMVariableNode() {variable=isCrouchWalking}.Value).Result
   )
  ]}
 CollapseNode_ContainedGraph {nodes [
   n1 = RigVMFunctionEntryNode()
   &RigUnit_ParentConstraint(
    AdvancedSettings=(RotationOrderForFilter=XZY)
    Child=(Type=Null,Name="dyn_hair_0_c_null")
    Parents=((Item=(Type=Bone,Name="head"),Weight=1))
    Weight=1
    ExecutePin=@n1.ExecutePin
   )
   > &RigUnit_ParentConstraint(
    AdvancedSettings=(RotationOrderForFilter=XZY)
    Child=(Type=Null,Name="dyn_hair_1_c_null")
    Parents=((Item=(Type=Bone,Name="head"),Weight=1))
    Weight=1
   )
   > &RigUnit_ParentConstraint(
    AdvancedSettings=(RotationOrderForFilter=XZY)
    Child=(Type=Null,Name="dyn_hair_2_c_null")
    Parents=((Item=(Type=Bone,Name="neck_02"),Weight=0.2962),(Item=(Type=Bone,Name="head"),Weight=0.318),(Item=(Type=Bone,Name="neck_01"),Weight=0.3859))
    Weight=1
   )
   > &RigUnit_ParentConstraint(
    AdvancedSettings=(RotationOrderForFilter=XZY)
    Child=(Type=Null,Name="dyn_hair_3_c_null")
    Parents=((Item=(Type=Bone,Name="neck_02"),Weight=0.0094),(Item=(Type=Bone,Name="clavicle_r"),Weight=0.0275),(Item=(Type=Bone,Name="clavicle_l"),Weight=0.0276),(Item=(Type=Bone,Name="neck_01"),Weight=0.1058),(Item=(Type=Bone,Name="spine_05"),Weight=0.8297))
    Weight=1
   )
   > &RigUnit_ParentConstraint(
    AdvancedSettings=(RotationOrderForFilter=XZY)
    Child=(Type=Null,Name="dyn_hair_4_c_null")
    Parents=((Item=(Type=Bone,Name="clavicle_l"),Weight=0.0685),(Item=(Type=Bone,Name="clavicle_r"),Weight=0.0727),(Item=(Type=Bone,Name="spine_05"),Weight=0.8588))
    Weight=1
   )
   > &RigUnit_ParentConstraint(
    AdvancedSettings=(RotationOrderForFilter=XZY)
    Child=(Type=Null,Name="dyn_hair_5_c_null")
    Parents=((Item=(Type=Bone,Name="spine_04"),Weight=0.008),(Item=(Type=Bone,Name="clavicle_l"),Weight=0.1495),(Item=(Type=Bone,Name="clavicle_r"),Weight=0.1799),(Item=(Type=Bone,Name="spine_05"),Weight=0.6626))
    Weight=1
   )
   > &RigUnit_ParentConstraint(
    AdvancedSettings=(RotationOrderForFilter=XZY)
    Child=(Type=Null,Name="dyn_hair_6_c_null")
    Parents="((Item=(Type=Bone,Name=\"spine_02\"),Weight=0.003),(Item=(Type=Bone,Name=\"spine_03\"),Weight=0.0036),(Item=(Type=Bone,Name=\"clavicle_r\"),Weight=0.1199),(Item=(Type=Bone,Name=\"clavicle_l\"),Weight=0.1494),(Item=(Type=Bone,Name=\"spine_04\"),Weight=0.1836),(Item=( ...(truncated)"
    Weight=1
   )
   > &RigUnit_ParentConstraint(
    AdvancedSettings=(RotationOrderForFilter=XZY)
    Child=(Type=Null,Name="dyn_hair_7_c_null")
    Parents="((Item=(Type=Bone,Name=\"spine_01\"),Weight=0.0228),(Item=(Type=Bone,Name=\"spine_03\"),Weight=0.0721),(Item=(Type=Bone,Name=\"spine_02\"),Weight=0.0723),(Item=(Type=Bone,Name=\"clavicle_l\"),Weight=0.0857),(Item=(Type=Bone,Name=\"clavicle_r\"),Weight=0.1285),(Item= ...(truncated)"
    Weight=1
   )
   > &RigUnit_ParentConstraint(
    AdvancedSettings=(RotationOrderForFilter=XZY)
    Child=(Type=Null,Name="dyn_hair_0_l_null")
    Parents=((Item=(Type=Bone,Name="head"),Weight=1.0))
    Weight=1
   )
   > &RigUnit_ParentConstraint(
    AdvancedSettings=(RotationOrderForFilter=XZY)
    Child=(Type=Null,Name="dyn_hair_1_l_null")
    Parents=((Item=(Type=Bone,Name="head"),Weight=1.0))
    Weight=1
   )
   > &RigUnit_ParentConstraint(
    AdvancedSettings=(RotationOrderForFilter=XZY)
    Child=(Type=Null,Name="dyn_hair_2_l_null")
    Parents=((Item=(Type=Bone,Name="neck_02"),Weight=0.3144),(Item=(Type=Bone,Name="neck_01"),Weight=0.3157),(Item=(Type=Bone,Name="head"),Weight=0.3698))
    Weight=1
   )
   > &RigUnit_ParentConstraint(
    AdvancedSettings=(RotationOrderForFilter=XZY)
    Child=(Type=Null,Name="dyn_hair_3_l_null")
    Parents=((Item=(Type=Bone,Name="neck_01"),Weight=0.0435),(Item=(Type=Bone,Name="clavicle_l"),Weight=0.1239),(Item=(Type=Bone,Name="spine_05"),Weight=0.8327))
    Weight=1
   )
   > &RigUnit_ParentConstraint(
    AdvancedSettings=(RotationOrderForFilter=XZY)
    Child=(Type=Null,Name="dyn_hair_4_l_null")
    Parents=((Item=(Type=Bone,Name="clavicle_r"),Weight=0.0202),(Item=(Type=Bone,Name="clavicle_l"),Weight=0.4364),(Item=(Type=Bone,Name="spine_05"),Weight=0.5434))
    Weight=1
   )
   > &RigUnit_ParentConstraint(
    AdvancedSettings=(RotationOrderForFilter=XZY)
    Child=(Type=Null,Name="dyn_hair_5_l_null")
    Parents=((Item=(Type=Bone,Name="clavicle_r"),Weight=0.0337),(Item=(Type=Bone,Name="spine_04"),Weight=0.0547),(Item=(Type=Bone,Name="spine_05"),Weight=0.4424),(Item=(Type=Bone,Name="clavicle_l"),Weight=0.4691))
    Weight=1
   )
   > &RigUnit_ParentConstraint(
    AdvancedSettings=(RotationOrderForFilter=XZY)
    Child=(Type=Null,Name="dyn_hair_6_l_null")
    Parents="((Item=(Type=Bone,Name=\"spine_02\"),Weight=0.0067),(Item=(Type=Bone,Name=\"spine_03\"),Weight=0.0089),(Item=(Type=Bone,Name=\"clavicle_r\"),Weight=0.0136),(Item=(Type=Bone,Name=\"spine_04\"),Weight=0.1713),(Item=(Type=Bone,Name=\"spine_05\"),Weight=0.3054),(Item=(T ...(truncated)"
    Weight=1
   )
   > &RigUnit_ParentConstraint(
    AdvancedSettings=(RotationOrderForFilter=XZY)
    Child=(Type=Null,Name="dyn_hair_7_l_null")
    Parents="((Item=(Type=Bone,Name=\"spine_01\"),Weight=0.0151),(Item=(Type=Bone,Name=\"clavicle_r\"),Weight=0.0233),(Item=(Type=Bone,Name=\"spine_03\"),Weight=0.0614),(Item=(Type=Bone,Name=\"spine_02\"),Weight=0.0666),(Item=(Type=Bone,Name=\"spine_05\"),Weight=0.2258),(Item=(T ...(truncated)"
    Weight=1
   )
   > &RigUnit_ParentConstraint(
    AdvancedSettings=(RotationOrderForFilter=XZY)
    Child=(Type=Null,Name="dyn_hair_0_r_null")
    Parents=((Item=(Type=Bone,Name="head"),Weight=1.0))
    Weight=1
   )
   > &RigUnit_ParentConstraint(
    AdvancedSettings=(RotationOrderForFilter=XZY)
    Child=(Type=Null,Name="dyn_hair_1_r_null")
    Parents=((Item=(Type=Bone,Name="head"),Weight=1.0))
    Weight=1
   )
   > &RigUnit_ParentConstraint(
    AdvancedSettings=(RotationOrderForFilter=XZY)
    Child=(Type=Null,Name="dyn_hair_2_r_null")
    Parents=((Item=(Type=Bone,Name="neck_02"),Weight=0.3144),(Item=(Type=Bone,Name="neck_01"),Weight=0.3157),(Item=(Type=Bone,Name="head"),Weight=0.3698))
    Weight=1
   )
   > &RigUnit_ParentConstraint(
    AdvancedSettings=(RotationOrderForFilter=XZY)
    Child=(Type=Null,Name="dyn_hair_3_r_null")
    Parents=((Item=(Type=Bone,Name="neck_01"),Weight=0.0435),(Item=(Type=Bone,Name="clavicle_r"),Weight=0.1239),(Item=(Type=Bone,Name="spine_05"),Weight=0.8327))
    Weight=1
   )
   > &RigUnit_ParentConstraint(
    AdvancedSettings=(RotationOrderForFilter=XZY)
    Child=(Type=Null,Name="dyn_hair_4_r_null")
    Parents=((Item=(Type=Bone,Name="spine_04"),Weight=0.0011),(Item=(Type=Bone,Name="clavicle_l"),Weight=0.0257),(Item=(Type=Bone,Name="clavicle_r"),Weight=0.4679),(Item=(Type=Bone,Name="spine_05"),Weight=0.5054))
    Weight=1
   )
   > &RigUnit_ParentConstraint(
    AdvancedSettings=(RotationOrderForFilter=XZY)
    Child=(Type=Null,Name="dyn_hair_5_r_null")
    Parents=((Item=(Type=Bone,Name="clavicle_l"),Weight=0.0375),(Item=(Type=Bone,Name="spine_04"),Weight=0.0525),(Item=(Type=Bone,Name="clavicle_r"),Weight=0.4522),(Item=(Type=Bone,Name="spine_05"),Weight=0.4579))
    Weight=1
   )
   > &RigUnit_ParentConstraint(
    AdvancedSettings=(RotationOrderForFilter=XZY)
    Child=(Type=Null,Name="dyn_hair_6_r_null")
    Parents="((Item=(Type=Bone,Name=\"spine_02\"),Weight=0.0179),(Item=(Type=Bone,Name=\"spine_03\"),Weight=0.0207),(Item=(Type=Bone,Name=\"clavicle_l\"),Weight=0.0243),(Item=(Type=Bone,Name=\"spine_04\"),Weight=0.2113),(Item=(Type=Bone,Name=\"spine_05\"),Weight=0.3211),(Item=(T ...(truncated)"
    Weight=1
   )
   > n2 = &RigUnit_ParentConstraint(
    AdvancedSettings=(RotationOrderForFilter=XZY)
    Child=(Type=Null,Name="dyn_hair_7_r_null")
    Parents="((Item=(Type=Bone,Name=\"spine_01\"),Weight=0.0079),(Item=(Type=Bone,Name=\"clavicle_l\"),Weight=0.0231),(Item=(Type=Bone,Name=\"spine_03\"),Weight=0.0496),(Item=(Type=Bone,Name=\"spine_02\"),Weight=0.051),(Item=(Type=Bone,Name=\"spine_05\"),Weight=0.247),(Item=(Typ ...(truncated)"
    Weight=1
   )
   RigVMFunctionReturnNode(ExecutePin=@n2.ExecutePin)
  ]}
 CollapseNode_ContainedGraph {nodes [
   n1 = RigVMFunctionEntryNode()
   RigVMFunction_ControlFlowBranch = &RigVMFunction_ControlFlowBranch(Condition=@n1.Condition, ExecuteContext=@n1.ExecuteContext)
   RigVMFunctionReturnNode(ExecuteContext=@RigVMFunction_ControlFlowBranch.Completed)
   ItemArray_5 = RigVMDispatchNode(Value=((Type=Bone,Name="dyn_hair_0_r"),(Type=Bone,Name="dyn_hair_1_r"),(Type=Bone,Name="dyn_hair_2_r"),(Type=Bone,Name="dyn_hair_3_r"),(Type=Bone,Name="dyn_hair_4_r"))) {template="DISPATCH_RigVMDispatch_Constant(io Value)"}
   ItemArray_8 = RigVMDispatchNode(Value=((Type=Control,Name="dyn_hair_0_r_ctrl"),(Type=Control,Name="dyn_hair_1_r_ctrl"),(Type=Control,Name="dyn_hair_2_r_ctrl"),(Type=Control,Name="dyn_hair_3_r_ctrl"),(Type=Control,Name="dyn_hair_4_r_ctrl"))) {template="DISPATCH_RigVMDispatch_Constant(io Value)"}
   SplineIK_1 = RigVMFunctionReferenceNode(
    Primary Axis=(X=1,Y=0,Z=0)
    Samples Per Segment=1
    Secondary Spline Direction=(X=0,Y=0,Z=0)
    Spline Mode=BSpline
    Stretch=true
    Up Axis=(X=0,Y=0,Z=0)
    debug=false
    Bones=@ItemArray_5.Value
    Controls=@ItemArray_8.Value
    ExecuteContext=@RigVMFunction_ControlFlowBranch.False
   )
   ItemArray_6 = RigVMDispatchNode(Value=((Type=Bone,Name="dyn_hair_0_l"),(Type=Bone,Name="dyn_hair_1_l"),(Type=Bone,Name="dyn_hair_2_l"),(Type=Bone,Name="dyn_hair_3_l"),(Type=Bone,Name="dyn_hair_4_l"))) {template="DISPATCH_RigVMDispatch_Constant(io Value)"}
   ItemArray_7 = RigVMDispatchNode(Value=((Type=Control,Name="dyn_hair_0_l_ctrl"),(Type=Control,Name="dyn_hair_1_l_ctrl"),(Type=Control,Name="dyn_hair_2_l_ctrl"),(Type=Control,Name="dyn_hair_3_l_ctrl"),(Type=Control,Name="dyn_hair_4_l_ctrl"))) {template="DISPATCH_RigVMDispatch_Constant(io Value)"}
   > SplineIK_2 = RigVMFunctionReferenceNode(
    Primary Axis=(X=1,Y=0,Z=0)
    Samples Per Segment=1
    Secondary Spline Direction=(X=0,Y=0,Z=0)
    Spline Mode=BSpline
    Stretch=true
    Up Axis=(X=0,Y=0,Z=0)
    debug=false
    Bones=@ItemArray_6.Value
    Controls=@ItemArray_7.Value
   )
   ItemArray_14 = RigVMDispatchNode(Value=((Type=Control,Name="dyn_hair_0_c_ctrl"),(Type=Control,Name="dyn_hair_1_c_ctrl"),(Type=Control,Name="dyn_hair_2_c_ctrl"),(Type=Control,Name="dyn_hair_3_c_ctrl"),(Type=Control,Name="dyn_hair_4_c_ctrl"))) {template="DISPATCH_RigVMDispatch_Constant(io Value)"}
   ItemArray_4 = RigVMDispatchNode(Value=((Type=Bone,Name="dyn_hair_0_c"),(Type=Bone,Name="dyn_hair_1_c"),(Type=Bone,Name="dyn_hair_2_c"),(Type=Bone,Name="dyn_hair_3_c"),(Type=Bone,Name="dyn_hair_4_c"))) {template="DISPATCH_RigVMDispatch_Constant(io Value)"}
   > SplineIK = RigVMFunctionReferenceNode(
    Primary Axis=(X=1,Y=0,Z=0)
    Samples Per Segment=1
    Secondary Spline Direction=(X=0,Y=0,Z=0)
    Spline Mode=BSpline
    Stretch=true
    Up Axis=(X=0,Y=0,Z=0)
    debug=false
    Controls=@ItemArray_14.Value
    Bones=@ItemArray_4.Value
   )
   ItemArray_10 = RigVMDispatchNode(Value="((Type=Bone,Name=\"dyn_hair_0_r\"),(Type=Bone,Name=\"dyn_hair_1_r\"),(Type=Bone,Name=\"dyn_hair_2_r\"),(Type=Bone,Name=\"dyn_hair_3_r\"),(Type=Bone,Name=\"dyn_hair_4_r\"),(Type=Bone,Name=\"dyn_hair_5_r\"),(Type=Bone,Name=\"dyn_hair_6_r\"),(Type=Bone,Name=\"dyn_hair_7_r\") ...(truncated)") {template="DISPATCH_RigVMDispatch_Constant(io Value)"}
   ItemArray_13 = RigVMDispatchNode(Value="((Type=Control,Name=\"dyn_hair_0_r_ctrl\"),(Type=Control,Name=\"dyn_hair_1_r_ctrl\"),(Type=Control,Name=\"dyn_hair_2_r_ctrl\"),(Type=Control,Name=\"dyn_hair_3_r_ctrl\"),(Type=Control,Name=\"dyn_hair_4_r_ctrl\"),(Type=Control,Name=\"dyn_hair_5_r_ctrl\"),(Type=Control,N ...(truncated)") {template="DISPATCH_RigVMDispatch_Constant(io Value)"}
   SplineIK_4 = RigVMFunctionReferenceNode(
    Primary Axis=(X=1,Y=0,Z=0)
    Samples Per Segment=1
    Secondary Spline Direction=(X=0,Y=0,Z=0)
    Spline Mode=BSpline
    Stretch=true
    Up Axis=(X=0,Y=0,Z=0)
    debug=false
    Bones=@ItemArray_10.Value
    Controls=@ItemArray_13.Value
    ExecuteContext=@RigVMFunction_ControlFlowBranch.True
   )
   ItemArray_11 = RigVMDispatchNode(Value="((Type=Bone,Name=\"dyn_hair_0_l\"),(Type=Bone,Name=\"dyn_hair_1_l\"),(Type=Bone,Name=\"dyn_hair_2_l\"),(Type=Bone,Name=\"dyn_hair_3_l\"),(Type=Bone,Name=\"dyn_hair_4_l\"),(Type=Bone,Name=\"dyn_hair_5_l\"),(Type=Bone,Name=\"dyn_hair_6_l\"),(Type=Bone,Name=\"dyn_hair_7_l\") ...(truncated)") {template="DISPATCH_RigVMDispatch_Constant(io Value)"}
   ItemArray_12 = RigVMDispatchNode(Value="((Type=Control,Name=\"dyn_hair_0_l_ctrl\"),(Type=Control,Name=\"dyn_hair_1_l_ctrl\"),(Type=Control,Name=\"dyn_hair_2_l_ctrl\"),(Type=Control,Name=\"dyn_hair_3_l_ctrl\"),(Type=Control,Name=\"dyn_hair_4_l_ctrl\"),(Type=Control,Name=\"dyn_hair_5_l_ctrl\"),(Type=Control,N ...(truncated)") {template="DISPATCH_RigVMDispatch_Constant(io Value)"}
   > SplineIK_5 = RigVMFunctionReferenceNode(
    Primary Axis=(X=1,Y=0,Z=0)
    Samples Per Segment=1
    Secondary Spline Direction=(X=0,Y=0,Z=0)
    Spline Mode=BSpline
    Stretch=true
    Up Axis=(X=0,Y=0,Z=0)
    debug=false
    Bones=@ItemArray_11.Value
    Controls=@ItemArray_12.Value
   )
   ItemArray_2 = RigVMDispatchNode(Value="((Type=Control,Name=\"dyn_hair_0_c_ctrl\"),(Type=Control,Name=\"dyn_hair_1_c_ctrl\"),(Type=Control,Name=\"dyn_hair_2_c_ctrl\"),(Type=Control,Name=\"dyn_hair_3_c_ctrl\"),(Type=Control,Name=\"dyn_hair_4_c_ctrl\"),(Type=Control,Name=\"dyn_hair_5_c_ctrl\"),(Type=Control,N ...(truncated)") {template="DISPATCH_RigVMDispatch_Constant(io Value)"}
   ItemArray_3 = RigVMDispatchNode(Value="((Type=Bone,Name=\"dyn_hair_0_c\"),(Type=Bone,Name=\"dyn_hair_1_c\"),(Type=Bone,Name=\"dyn_hair_2_c\"),(Type=Bone,Name=\"dyn_hair_3_c\"),(Type=Bone,Name=\"dyn_hair_4_c\"),(Type=Bone,Name=\"dyn_hair_5_c\"),(Type=Bone,Name=\"dyn_hair_6_c\"),(Type=Bone,Name=\"dyn_hair_7_c\") ...(truncated)") {template="DISPATCH_RigVMDispatch_Constant(io Value)"}
   > SplineIK_3 = RigVMFunctionReferenceNode(
    Primary Axis=(X=1,Y=0,Z=0)
    Samples Per Segment=1
    Secondary Spline Direction=(X=0,Y=0,Z=0)
    Spline Mode=BSpline
    Stretch=true
    Up Axis=(X=0,Y=0,Z=0)
    debug=false
    Controls=@ItemArray_2.Value
    Bones=@ItemArray_3.Value
   )
  ]}
 CollapseNode_ContainedGraph {nodes [
   RigVMFunctionEntryNode()
   > RigVMFunction_Sequence_1 = RigVMAggregateNode()
   DeltaRotation v4_2 = RigVMFunctionReferenceNode(
    Rotation Order=ZYX
    child=(Type=Bone,Name="upperarm_l")
    debugRotation=false
    onlyTwist=false
    outputMinMax=(X=0,Y=1)
    parent=(Type=Null,Name="upperarm_l_null")
    premult=1
    rotationMinMax=(X=0,Y=-45)
    triggerAxis=Y
    twistAxis=(X=1,Y=0,Z=0)
    ExecuteContext=@RigVMFunction_Sequence_1.A
   )
   > &RigUnit_ModifyTransforms(
    ItemToModify="((Item=(Type=Bone,Name=\"dyn_hairShoulder_l\"),Transform=(Rotation=(X=0,Y=-0.1736,Z=0,W=0.9848),Translation=(X=0,Y=-1,Z=0),Scale3D=(X=1,Y=1,Z=1))),(Item=(Type=Null,Name=\"dyn_hair_2_l_null\"),Transform=(Rotation=(X=0,Y=0,Z=0,W=1),Translation=(X=0,Y=-3,Z=4),Sca ...(truncated)"
    WeightMinimum=0
    Weight=@"DeltaRotation v4_2".outputWeight
   )
   > n1 = &RigUnit_SphericalPoseReader(
    ActiveRegionScaleFactors=(PositiveWidth=0.5,NegativeWidth=0.5,PositiveHeight=0.5,NegativeHeight=0.5)
    ActiveRegionSize=0.12
    Debug=(bDrawDebug=false,bDraw2D=False,bDrawLocalAxes=False,DebugScale=25,DebugSegments=20,DebugThickness=0.25)
    DriverAxis=(X=1,Y=0,Z=0)
    DriverItem=(Type=Bone,Name="clavicle_l")
    FalloffRegionScaleFactors=(PositiveWidth=0.5,NegativeWidth=0.5,PositiveHeight=0.5,NegativeHeight=0.5)
    FalloffSize=0.3
    FlipHeightScaling=False
    FlipWidthScaling=False
    OptionalParentItem=(Type=Bone,Name="None")
    RotationOffset=(X=0,Y=90,Z=140)
   )
   > &RigUnit_ModifyTransforms(
    ItemToModify="((Item=(Type=Null,Name=\"dyn_hair_3_l_null\"),Transform=(Rotation=(X=0,Y=0,Z=0,W=1),Translation=(X=0,Y=3,Z=5.5),Scale3D=(X=1,Y=1,Z=1))),(Item=(Type=Null,Name=\"dyn_hair_4_l_null\"),Transform=(Rotation=(X=0,Y=0,Z=0,W=1),Translation=(X=0,Y=-1.8,Z=6.5),Scale3D=(X ...(truncated)"
    WeightMinimum=0
    Weight=@n1.OutputParam
   )
   DeltaRotation v4_3 = RigVMFunctionReferenceNode(
    Rotation Order=ZYX
    child=(Type=Bone,Name="upperarm_r")
    debugRotation=false
    onlyTwist=false
    outputMinMax=(X=0,Y=1)
    parent=(Type=Null,Name="upperarm_r_null")
    premult=1
    rotationMinMax=(X=0,Y=-45)
    triggerAxis=Y
    twistAxis=(X=1,Y=0,Z=0)
    ExecuteContext=@RigVMFunction_Sequence_1.B
   )
   > &RigUnit_ModifyTransforms(
    ItemToModify="((Item=(Type=Bone,Name=\"dyn_hairShoulder_r\"),Transform=(Rotation=(X=0,Y=-0.2588,Z=0,W=0.9659),Translation=(X=0,Y=0,Z=0),Scale3D=(X=1,Y=1,Z=1))),(Item=(Type=Null,Name=\"dyn_hair_2_r_null\"),Transform=(Rotation=(X=0,Y=0,Z=0,W=1),Translation=(X=1,Y=4.5,Z=3),Sca ...(truncated)"
    WeightMinimum=0
    Weight=@"DeltaRotation v4_3".outputWeight
   )
   > n2 = &RigUnit_SphericalPoseReader(
    ActiveRegionScaleFactors=(PositiveWidth=0.5,NegativeWidth=0.5,PositiveHeight=0.5,NegativeHeight=0.5)
    ActiveRegionSize=0.12
    Debug=(bDrawDebug=false,bDraw2D=False,bDrawLocalAxes=False,DebugScale=25,DebugSegments=20,DebugThickness=0.25)
    DriverAxis=(X=-1,Y=0,Z=0)
    DriverItem=(Type=Bone,Name="clavicle_r")
    FalloffRegionScaleFactors=(PositiveWidth=0.5,NegativeWidth=0.5,PositiveHeight=0.5,NegativeHeight=0.5)
    FalloffSize=0.3
    FlipHeightScaling=False
    FlipWidthScaling=False
    OptionalParentItem=(Type=Bone,Name="None")
    RotationOffset=(X=0,Y=-90,Z=140)
   )
   > &RigUnit_ModifyTransforms(
    ItemToModify=((Item=(Type=Null,Name="dyn_hair_3_r_null"),Transform=(Translation=(X=0,Y=-3,Z=5.5),Scale3D=(X=1,Y=1,Z=1))),(Item=(Type=Null,Name="dyn_hair_4_r_null"),Transform=(Rotation=(X=0,Y=0,Z=0,W=1),Translation=(X=0,Y=-1.8,Z=6.5),Scale3D=(X=1,Y=1,Z=1))))
    WeightMinimum=0
    Weight=@n2.OutputParam
   )
   RigVMFunctionReturnNode(ExecuteContext=@RigVMFunction_Sequence_1.C)
  ]}
 CollapseNode_ContainedGraph {nodes [
   n1 = RigVMFunctionEntryNode()
   RigVMFunction_Sequence = &RigVMFunction_Sequence(ExecuteContext=@n1.ExecuteContext)
   n2 = &RigVMFunction_AlphaInterp(
    InterpSpeedDecreasing=10
    InterpSpeedIncreasing=10
    Scale=1
    bClampResult=true
    Value=&RigVMFunction_MathBoolToFloat(Value=RigVMVariableNode() {variable=isTargeting}.Value).Result
   )
   RigVMVariableNode(Value=@n2.Result, ExecuteContext=@RigVMFunction_Sequence.A) {variable=isTargetingLerp}
   > RigVMFunctionReturnNode()
   n3 = &RigVMFunction_AnimEvalRichCurve(
    Curve=(EditorCurveData=(Keys=((InterpMode=RCIM_Cubic),(InterpMode=RCIM_Cubic,Time=0.5,Value=1),(Time=0.8))))
    SourceMaximum=2
    SourceMinimum=0
    TargetMinimum=0
    Value=&RigVMFunction_MathDoubleRemap(
     SourceMaximum=75
     SourceMinimum=45
     TargetMaximum=1
     TargetMinimum=0
     Value=&RigUnit_GetRelativeTransformForItem(Child=(Type=Bone,Name="pelvis"), Parent=(Type=Bone,Name="root"), bChildInitial=False, bParentInitial=False).RelativeTransform.Translation.Z
    ).Result
   )
   Subtract = &RigVMFunction_MathDoubleSub(A=1, B=&RigVMFunction_MathBoolToFloat(Value=RigVMVariableNode() {variable=isSkyFloating}.Value).Result)
   Multiply_6 = RigVMAggregateNode(
    C=@n3.Result
    A=&RigVMFunction_MathBoolToFloat(Value=RigVMVariableNode() {variable=isCrouching}.Value).Result
    B=@Subtract.Result
   )
   RigVMVariableNode(Value=@Multiply_6.Result, ExecuteContext=@RigVMFunction_Sequence.B) {variable=isCrouchWalking}
  ]}
 Multiply_1_ContainedGraph {nodes [
   n1 = RigVMFunctionEntryNode()
   RigVMFunctionReturnNode(Result=&RigVMFunction_MathDoubleMul(B=@n1.C, A=&RigVMFunction_MathDoubleMul(A=@n1.A, B=@n1.B).Result).Result)
  ]}
 Multiply_5_ContainedGraph {nodes [
   n1 = RigVMFunctionEntryNode()
   RigVMFunctionReturnNode(Result=&RigVMFunction_MathDoubleMul(B=@n1.C, A=&RigVMFunction_MathDoubleMul(A=@n1.A, B=@n1.B).Result).Result)
  ]}
 Multiply_7_ContainedGraph {nodes [
   n1 = RigVMFunctionEntryNode()
   RigVMFunctionReturnNode(Result=&RigVMFunction_MathDoubleMul(B=@n1.C, A=&RigVMFunction_MathDoubleMul(A=@n1.A, B=@n1.B).Result).Result)
  ]}
 Multiply_8_ContainedGraph {nodes [
   n1 = RigVMFunctionEntryNode()
   RigVMFunctionReturnNode(Result=&RigVMFunction_MathDoubleMul(B=@n1.C, A=&RigVMFunction_MathDoubleMul(A=@n1.A, B=@n1.B).Result).Result)
  ]}
 New Function_ContainedGraph {nodes [
   n1 = RigVMFunctionEntryNode()
   RigVMFunction_MathBoolToInteger = &RigVMFunction_MathBoolToInteger(Value=@n1.onlyTwist)
   RigVMFunction_MathQuaternionSwingTwist = &RigVMFunction_MathQuaternionSwingTwist(
    twistAxis=@n1.twistAxis
    Input=&RigUnit_GetRelativeTransformForItem(bChildInitial=False, bParentInitial=False, child=@n1.child, parent=@n1.parent).RelativeTransform.Rotation
   )
   n2 = RigVMDispatchNode(
    Values=((),())
    Index=@RigVMFunction_MathBoolToInteger.Result
    Values.0=@RigVMFunction_MathQuaternionSwingTwist.Swing
    Values.1=@RigVMFunction_MathQuaternionSwingTwist.Twist
   ) {template="DISPATCH_RigVMDispatch_SelectInt32(in Index,in Values,out Result)"}
   RigVMFunction_MathQuaternionToEuler = &RigVMFunction_MathQuaternionToEuler(Value=@n2.Result, RotationOrder=@n1.Rotation Order)
   > DISPATCH_RigVMDispatch_Print = RigVMDispatchNode(ScreenColor=(R=1,G=1,B=1,A=1), ScreenDuration=1.5, Enabled=@n1.debugRotation, Value=@RigVMFunction_MathQuaternionToEuler.Result) {template="DISPATCH_RigVMDispatch_Print(in Prefix,in Value,in Enabled,in ScreenDuration,in ScreenColor)"}
   Cast = RigVMDispatchNode(Value=@n1.triggerAxis) {template="DISPATCH_RigVMDispatch_CastEnumToInt(in Value,out Result)"}
   DISPATCH_RigVMDispatch_SelectInt32 = RigVMDispatchNode(
    Values=(0,0,0)
    Index=@Cast.Result
    Values.0=@RigVMFunction_MathQuaternionToEuler.Result.X
    Values.1=@RigVMFunction_MathQuaternionToEuler.Result.Y
    Values.2=@RigVMFunction_MathQuaternionToEuler.Result.Z
   ) {template="DISPATCH_RigVMDispatch_SelectInt32(in Index,in Values,out Result)"}
   Remap = &RigVMFunction_MathDoubleRemap(
    bClamp=False
    Value=@DISPATCH_RigVMDispatch_SelectInt32.Result
    TargetMinimum=@n1.outputMinMax.X
    TargetMaximum=@n1.outputMinMax.Y
    SourceMinimum=@n1.rotationMinMax.X
    SourceMaximum=@n1.rotationMinMax.Y
   )
   Absolute = &RigVMFunction_MathDoubleAbs(Value=@Remap.Result)
   Multiply = &RigVMFunction_MathDoubleMul(
    A=&RigVMFunction_MathDoubleClamp(Value=RigVMDispatchNode(True=@Absolute.Result, Condition=@n1.absoluteValue, False=@Remap.Result) {template="DISPATCH_RigVMDispatch_If(in Condition,in True,in False,out Result)"}.Result).Result
    B=@n1.premult
   )
   > RigVMFunctionReturnNode(outputQuat=@n2.Result, outputWeight=@Multiply.Result)
  ]}
 New Function_ContainedGraph {nodes [
   n1 = RigVMFunctionEntryNode()
   RigVMFunction_Sequence = &RigVMFunction_Sequence(ExecuteContext=@n1.ExecuteContext)
   n2 = &RigUnit_SphericalPoseReader(
    ActiveRegionScaleFactors=(PositiveWidth=1,NegativeWidth=1,PositiveHeight=2,NegativeHeight=1)
    ActiveRegionSize=0.05
    Debug=(bDrawDebug=false,bDraw2D=False,bDrawLocalAxes=False,DebugScale=25,DebugSegments=20,DebugThickness=0.25)
    DriverAxis=(X=-1,Y=0,Z=0)
    DriverItem=(Type=Bone,Name="upperarm_r")
    FalloffRegionScaleFactors=(PositiveWidth=1,NegativeWidth=1,PositiveHeight=1,NegativeHeight=1)
    FalloffSize=0.25
    FlipHeightScaling=False
    FlipWidthScaling=False
    OptionalParentItem=(Type=Bone,Name="None")
    RotationOffset=(X=-120,Y=20,Z=20)
    ExecutePin=@RigVMFunction_Sequence.A
   )
   RigVMFunction_MathBoolNot = &RigVMFunction_MathBoolNot(Value=RigVMVariableNode() {variable=isCrouching}.Value)
   Multiply_1 = RigVMAggregateNode(
    A=&RigVMFunction_MathBoolToFloat(Value=@RigVMFunction_MathBoolNot.Result).Result
    C=&RigVMFunction_MathBoolToFloat(Value=RigVMVariableNode() {variable=PubIsGunOut}.Value).Result
    B=@n2.OutputParam
   )
   > &RigUnit_ModifyTransforms(
    ItemToModify=((Item=(Type=Bone,Name="base_hair_0_A_r"),Transform=(Rotation=(X=0,Y=0,Z=0,W=1),Translation=(X=0,Y=-1,Z=0),Scale3D=(X=1,Y=1,Z=1))))
    WeightMinimum=0
    Weight=@Multiply_1.Result
   )
   > DeltaRotation v4 = RigVMFunctionReferenceNode(
    Rotation Order=ZYX
    child=(Type=Bone,Name="head")
    debugRotation=false
    outputMinMax=(X=0,Y=1)
    parent=(Type=Null,Name="head_sp_null_2")
    premult=1
    rotationMinMax=(X=-7,Y=17)
    triggerAxis=Z
    twistAxis=(X=1,Y=0,Z=0)
   )
   n3 = &RigVMFunction_MathBoolNot(Value=RigVMVariableNode() {variable=isCrouching}.Value)
   AlphaInterp = &RigVMFunction_AlphaInterp(
    InterpSpeedDecreasing=8
    InterpSpeedIncreasing=10
    Scale=1
    bClampResult=False
    Value=&RigVMFunction_MathDoubleMul(A=@"DeltaRotation v4".outputWeight, B=&RigVMFunction_MathBoolToFloat(Value=@n3.Result).Result).Result
   )
   > &RigUnit_ModifyTransforms(
    ItemToModify=((Item=(Type=Bone,Name="base_hair_0_A_r"),Transform=(Rotation=(X=0,Y=0,Z=-0.2164,W=0.9763),Translation=(X=-0.5,Y=-3,Z=0),Scale3D=(X=1,Y=1,Z=1))))
    WeightMinimum=0
    Weight=@AlphaInterp.Result
   )
   > DeltaRotation v4_1 = RigVMFunctionReferenceNode(
    Rotation Order=ZYX
    child=(Type=Bone,Name="head")
    debugRotation=false
    outputMinMax=(X=0,Y=1)
    parent=(Type=Null,Name="head_sp_null_2")
    premult=1
    rotationMinMax=(X=-30,Y=12)
    triggerAxis=Z
    twistAxis=(X=1,Y=0,Z=0)
   )
   n4 = RigVMVariableNode() {variable=isCrouching}
   n5 = &RigVMFunction_AlphaInterp(
    InterpSpeedDecreasing=8
    InterpSpeedIncreasing=8
    Scale=1
    bClampResult=False
    Value=&RigVMFunction_MathDoubleMul(A=@"DeltaRotation v4_1".outputWeight, B=&RigVMFunction_MathBoolToFloat(Value=@n4.Value).Result).Result
   )
   > &RigUnit_ModifyTransforms(
    ItemToModify=((Item=(Type=Bone,Name="base_hair_0_A_r"),Transform=(Rotation=(X=0,Y=0,Z=-0.2164,W=0.9763),Translation=(X=0,Y=-5,Z=0),Scale3D=(X=1,Y=1,Z=1))))
    WeightMinimum=0
    Weight=@n5.Result
   )
   > DeltaRotation v4_3 = RigVMFunctionReferenceNode(
    Rotation Order=ZYX
    child=(Type=Bone,Name="head")
    debugRotation=false
    outputMinMax=(X=0,Y=1)
    parent=(Type=Null,Name="head_sp_null_2")
    premult=1
    rotationMinMax=(X=-47,Y=-18)
    triggerAxis=Z
    twistAxis=(X=1,Y=0,Z=0)
   )
   Multiply_5 = RigVMAggregateNode(
    A=@"DeltaRotation v4_3".outputWeight
    B=&RigVMFunction_MathBoolToFloat(Value=RigVMVariableNode() {variable=isCrouching}.Value).Result
    C=RigVMVariableNode() {variable=isCrouchWalking}.Value
   )
   n6 = &RigVMFunction_AlphaInterp(InterpSpeedDecreasing=8, InterpSpeedIncreasing=8, Scale=1, bClampResult=False, Value=@Multiply_5.Result)
   > &RigUnit_ModifyTransforms(
    ItemToModify=((Item=(Type=Bone,Name="base_hair_0_A_r"),Transform=(Rotation=(X=0,Y=0,Z=-0.342,W=0.9397),Translation=(X=0,Y=-5,Z=0),Scale3D=(X=1,Y=1,Z=1))))
    WeightMinimum=0
    Weight=@n6.Result
   )
   > DeltaRotation v4_2 = RigVMFunctionReferenceNode(
    Rotation Order=ZYX
    child=(Type=Bone,Name="head")
    debugRotation=false
    outputMinMax=(X=0,Y=1)
    parent=(Type=Null,Name="head_sp_null_2")
    premult=1
    rotationMinMax=(X=-30,Y=-60)
    triggerAxis=Z
    twistAxis=(X=1,Y=0,Z=0)
   )
   n7 = RigVMVariableNode() {variable=isTargeting}
   Multiply_3 = RigVMAggregateNode(
    A=@"DeltaRotation v4_2".outputWeight
    B=&RigVMFunction_MathBoolToFloat(Value=RigVMVariableNode() {variable=isCrouching}.Value).Result
    C=&RigVMFunction_MathBoolToFloat(Value=&RigVMFunction_MathBoolNot(Value=@n7.Value).Result).Result
   )
   n8 = &RigVMFunction_AlphaInterp(InterpSpeedDecreasing=8, InterpSpeedIncreasing=8, Scale=1, bClampResult=False, Value=@Multiply_3.Result)
   > &RigUnit_ModifyTransforms(
    ItemToModify=((Item=(Type=Bone,Name="base_hair_0_A_r"),Transform=(Rotation=(X=0,Y=0,Z=0.0872,W=0.9962),Translation=(X=0,Y=0,Z=0),Scale3D=(X=1,Y=1,Z=1))))
    WeightMinimum=0
    Weight=@n8.Result
   )
   > DeltaRotation v4_4 = RigVMFunctionReferenceNode(
    Rotation Order=ZYX
    child=(Type=Bone,Name="head")
    debugRotation=false
    outputMinMax=(X=0,Y=1)
    parent=(Type=Null,Name="head_sp_null_2")
    premult=1
    rotationMinMax=(X=-30,Y=-60)
    triggerAxis=Z
    twistAxis=(X=1,Y=0,Z=0)
   )
   n9 = &RigVMFunction_AlphaInterp(InterpSpeedDecreasing=8, InterpSpeedIncreasing=8, Scale=1, bClampResult=False, Value=@"DeltaRotation v4_4".outputWeight)
   > &RigUnit_ModifyTransforms(
    ItemToModify=((Item=(Type=Bone,Name="base_hair_0_A_r"),Transform=(Rotation=(X=0,Y=0,Z=0.342,W=0.9397),Translation=(X=0,Y=0,Z=0),Scale3D=(X=1,Y=1,Z=1))))
    WeightMinimum=0
    Weight=@n9.Result
   )
   RigVMFunctionReturnNode(ExecuteContext=@RigVMFunction_Sequence.B)
  ]}
 RigVMFunctionLibrary {nodes [
   AlterTransform = RigVMCollapseNode(ExecuteContext=(), inheritsTransform=false, transforms=(), weights=1)
   DeltaRotation v4 = RigVMCollapseNode(
    ExecuteContext=()
    Rotation Order=ZYX
    child=(Type=Bone,Name="None")
    outputMinMax=(X=0,Y=1)
    parent=(Type=Bone,Name="None")
    premult=1
    rotationMinMax=(X=0,Y=0)
    twistAxis=(X=1,Y=0,Z=0)
   )
   FrontHairAdjustments = RigVMCollapseNode(ExecuteContext=())
  ]}
 RigVMFunction_Sequence_1_ContainedGraph {nodes [
   n1 = RigVMFunctionEntryNode()
   RigVMFunction_Sequence = &RigVMFunction_Sequence(ExecuteContext=@n1.ExecuteContext)
   n2 = &RigVMFunction_Sequence(ExecuteContext=@RigVMFunction_Sequence.B)
   RigVMFunctionReturnNode(A=@RigVMFunction_Sequence.A, C=@n2.A, E=@n2.B)
  ]}
 RigVMFunction_Sequence_2_ContainedGraph {nodes [
   n1 = RigVMFunctionEntryNode()
   n2 = &RigVMFunction_Sequence(ExecuteContext=@n1.ExecuteContext)
   n3 = &RigVMFunction_Sequence(ExecuteContext=@n2.B)
   RigVMFunctionReturnNode(A=@n2.A, B=@n3.A, C=@n3.B)
  ]}
 RigVMModel {nodes [
   RigUnit_BeginExecution = &RigUnit_BeginExecution()
   LERP Bools = RigVMCollapseNode(ExecuteContext=@RigUnit_BeginExecution.ExecutePin)
   hair = RigVMCollapseNode(ExecutePin=@"LERP Bools".ExecuteContext)
   FrontHairAdjustments = RigVMFunctionReferenceNode(ExecuteContext=@hair.ExecutePin)
   setFKCtrl = RigVMFunctionReferenceNode(ExecuteContext=(), setFK_Ctrl=false)
  ]}
 RigVMModel Construction Graph {nodes [
   n1 = &RigUnit_PrepareForExecution()
   ItemArray = RigVMDispatchNode(Value=((Type=Bone,Name="head"))) {template="DISPATCH_RigVMDispatch_Constant(io Value)"}
   addNullObject = RigVMFunctionReferenceNode(
    Options_Null="(pelvis_1_#1=False,spine_3_#2=False,clavicle_5_#3=False,arm_7_#4=False,leg_9_#5=False,head_11_46DC15EF465C ...(truncated)"
    Custom_Joints=@ItemArray.Value
    ExecuteContext=@n1.ExecutePin
   )
   > &RigUnit_HierarchyAddNull(
    Name=head_sp_null
    Parent=(Type=Bone,Name="spine_05")
    Space=LocalSpace
    Transform=&RigUnit_GetRelativeTransformForItem(Child=(Type=Bone,Name="head"), Parent=(Type=Bone,Name="spine_05"), bChildInitial=true, bParentInitial=true).RelativeTransform
   )
   addFKCtrl = RigVMFunctionReferenceNode(
    Custom_FKCtrl=()
    ExecuteContext=()
    Options_FKCtrl="(pelvis_1_#1=true,spine_3_#2=true,clavicle_5_#3=true,arm_7_#4=true,leg_9_#5=False,head_11_46DC15EF465CD297 ...(truncated)"
    Shape_Settings="(pelvis_2_#6=(bVisible=True,Name=\"Hexagon_Thin\",Color=(R=0.0117,G=0.8281,B=1,A=1),Transform=(Rotation=(X=0,Y=-0.7071,Z=0,W=0.7071),Translation=(X=2,Y=0,Z=0),Scale3D=(X=5,Y=5,Z=7))),spine_4_#7=(bVi ...(truncated)"
   )
  ]}
]
hierarchy [
 root {loc=<0 0 0>, rot=<0 0 0 1>}
 pelvis {parent=root, loc=<0 -0.5301 98.0571>, rot=<0.0127 0.707 -0.0127 -0.707>}
 spine_01 {parent=pelvis, loc=<6.1208 -0.9328 0>, rot=<0 0 0.034 -0.9994>}
 spine_02 {parent=spine_01, loc=<6.5997 -0.2095 0>, rot=<0 0 0.0238 -0.9997>}
 spine_03 {parent=spine_02, loc=<6.4914 -0.1237 0>, rot=<0 0 0.0149 -0.9999>}
 spine_04 {parent=spine_03, loc=<6.9598 0.0794 0>, rot=<0 0 -0.0132 -0.9999>}
 spine_05 {parent=spine_04, loc=<6.9987 0 0>, rot=<0 0 -0.0015 -1>}
 neck_01 {parent=spine_05, loc=<11.3138 0.1529 0>, rot=<0 0 -0.2003 -0.9797>}
 neck_02 {parent=neck_01, loc=<4.5627 0 0>, rot=<0 0 0.0504 -0.9987>}
 head {parent=neck_02, loc=<5.3914 0 0>, rot=<0 0 0.115 -0.9934>}
 base_hair_1_B_r {parent=head, loc=<13.0905 8.4956 2.748>, rot=<0.1596 0.4148 0.8947 0.0441>}
 dyn_hair_1_B_r {parent=base_hair_1_B_r, loc=<0 0 0>, rot=<0 0 0 1>}
 dyn_hair_2_B_r {parent=dyn_hair_1_B_r, loc=<5.5 0 0>, rot=<0 0.0096 0 -1>}
 dyn_hair_3_B_r {parent=dyn_hair_2_B_r, loc=<6 0 0>, rot=<0 0 0 1>}
 base_hair_0_c {parent=head, loc=<5.8985 -5.0186 0>, rot=<-0.0192 -0.7068 0.7068 0.0192>}
 dyn_hair_0_c {parent=base_hair_0_c, loc=<0 0 0>, rot=<0 0 0 1>}
 dyn_hair_1_c {parent=dyn_hair_0_c, loc=<4.6778 0 0.2937>, rot=<0 0.1876 0 -0.9822>}
 dyn_hair_2_c {parent=dyn_hair_1_c, loc=<5.0619 0 0>, rot=<0 0.075 0 -0.9972>}
 dyn_hair_3_c {parent=dyn_hair_2_c, loc=<5.0683 0 0>, rot=<0 0.0148 0 -0.9999>}
 dyn_hair_4_c {parent=dyn_hair_3_c, loc=<5.406 0 0>, rot=<0 -0.1899 0 -0.9818>}
 dyn_hair_5_c {parent=dyn_hair_4_c, loc=<5.0464 0 0>, rot=<0 -0.0863 0 -0.9963>}
 dyn_hair_6_c {parent=dyn_hair_5_c, loc=<5.0533 0 0>, rot=<0 0 0 1>}
 dyn_hair_7_c {parent=dyn_hair_6_c, loc=<5.0461 0 0>, rot=<0 0 0 1>}
 base_hair_0_r {parent=head, loc=<5.8985 -5.0186 5.6>, rot=<-0.0193 -0.7068 0.7068 0.0193>}
 dyn_hair_0_r {parent=base_hair_0_r, loc=<0 0 0>, rot=<0 0 0 1>}
 dyn_hair_1_r {parent=dyn_hair_0_r, loc=<4.6778 0 0.2938>, rot=<0 0.1876 0 -0.9822>}
 dyn_hair_2_r {parent=dyn_hair_1_r, loc=<5.0621 0 0>, rot=<0 0.0751 0 -0.9972>}
 dyn_hair_3_r {parent=dyn_hair_2_r, loc=<5.0678 0 0>, rot=<0 0.0148 0 -0.9999>}
 dyn_hair_4_r {parent=dyn_hair_3_r, loc=<5.4062 0 0>, rot=<0 -0.1899 0 -0.9818>}
 dyn_hair_5_r {parent=dyn_hair_4_r, loc=<5.0465 0 0>, rot=<0 -0.0863 0 -0.9963>}
 dyn_hair_6_r {parent=dyn_hair_5_r, loc=<5.053 0 0>, rot=<0 0 0 1>}
 dyn_hair_7_r {parent=dyn_hair_6_r, loc=<5.046 0 0>, rot=<0 0 0 1>}
 dyn_hairShoulder_r {parent=dyn_hair_1_r, loc=<3.597 -0.08 -1.645>, rot=<0 -0.6735 0 -0.7392>}
 dyn_hairShoulder_r_end {parent=dyn_hairShoulder_r, loc=<2.5 0 0>, rot=<0 0 0 1>}
 base_hair_0_l {parent=head, loc=<5.8985 -5.0186 -5.6>, rot=<0.0007 -0.7071 0.7071 -0.0007>}
 dyn_hair_0_l {parent=base_hair_0_l, loc=<0 0 0>, rot=<0 0 0 1>}
 dyn_hair_1_l {parent=dyn_hair_0_l, loc=<4.6869 0 0.0295>, rot=<0 0.1876 0 -0.9822>}
 dyn_hair_2_l {parent=dyn_hair_1_l, loc=<5.0551 0 -0.2424>, rot=<0 0.0469 0 -0.9989>}
 dyn_hair_3_l {parent=dyn_hair_2_l, loc=<5.0683 0 0>, rot=<0 0.0148 0 -0.9999>}
 dyn_hair_4_l {parent=dyn_hair_3_l, loc=<5.406 0 0>, rot=<0 -0.1899 0 -0.9818>}
 dyn_hair_5_l {parent=dyn_hair_4_l, loc=<5.0464 0 0>, rot=<0 -0.0863 0 -0.9963>}
 dyn_hair_6_l {parent=dyn_hair_5_l, loc=<5.0533 0 0>, rot=<0 0 0 1>}
 dyn_hair_7_l {parent=dyn_hair_6_l, loc=<5.0461 0 0>, rot=<0 0 0 1>}
 dyn_hairShoulder_l {parent=dyn_hair_1_l, loc=<3.5974 0.0804 -1.6446>, rot=<0 -0.6735 0 -0.7392>}
 dyn_hairShoulder_l_end {parent=dyn_hairShoulder_l, loc=<2.5 0 0>, rot=<0 0 0 1>}
 base_braid_1_r {parent=dyn_hair_0_l, loc=<4.3818 1.038 1.4116>, rot=<0.3739 0.1504 0.1023 -0.9094>}
 dyn_braid_1_r {parent=base_braid_1_r, loc=<0 0 0>, rot=<0 0 0 1>}
 dyn_braid_2_r {parent=dyn_braid_1_r, loc=<4.4051 0 0>, rot=<-0.0115 0.1813 0.062 -0.9814>}
 dyn_braid_3_r {parent=dyn_braid_2_r, loc=<5.2008 0 0>, rot=<-0.1349 -0.1037 0.083 -0.9819>}
 dyn_braid_4_r {parent=dyn_braid_3_r, loc=<6.8924 0 0>, rot=<0 0 0 1>}
 base_braid_1_c {parent=dyn_hair_0_l, loc=<4.7746 1.9753 -0.4566>, rot=<0.4198 0.1463 0.0149 -0.8956>}
 dyn_braid_1_c {parent=base_braid_1_c, loc=<0 0 0>, rot=<0 0 0 1>}
 dyn_braid_2_c {parent=dyn_braid_1_c, loc=<4.3772 0 0>, rot=<-0.082 0.2359 0.1101 -0.962>}
 dyn_braid_3_c {parent=dyn_braid_2_c, loc=<6.1921 0 0>, rot=<-0.0536 -0.1817 0.1587 -0.969>}
 dyn_braid_4_c {parent=dyn_braid_3_c, loc=<6.629 0 0>, rot=<-0.0046 -0.1794 0.0244 -0.9835>}
 dyn_braid_5_c {parent=dyn_braid_4_c, loc=<5.3239 0 0>, rot=<0 0 0 1>}
 base_braid_1_l {parent=dyn_hair_0_l, loc=<4.9566 1.8254 -2.3409>, rot=<0.7794 0.0968 -0.0914 -0.6122>}
 dyn_braid_1_l {parent=base_braid_1_l, loc=<0 0 0>, rot=<0 0 0 1>}
 dyn_braid_2_l {parent=dyn_braid_1_l, loc=<5.6181 0 0>, rot=<-0.0752 0.0229 0.363 -0.9285>}
 dyn_braid_3_l {parent=dyn_braid_2_l, loc=<6.9181 0 0>, rot=<-0.0565 -0.181 0.1699 -0.967>}
 dyn_braid_4_l {parent=dyn_braid_3_l, loc=<6.8266 0 0>, rot=<0 0 0 1>}
 base_hair_0_A_r {parent=head, loc=<5.5875 3.7938 7.4192>, rot=<-0.0472 0.0809 0.9956 0.0057>}
 dyn_hair_0_A_r {parent=base_hair_0_A_r, loc=<0 0 0>, rot=<0 0 0 1>}
 dyn_hair_1_A_r {parent=dyn_hair_0_A_r, loc=<7.7976 0 -0.0398>, rot=<0 -0.0533 0 -0.9986>}
 dyn_hair_2_A_r {parent=dyn_hair_1_A_r, loc=<5.6995 0 0>, rot=<0 0.1908 0 -0.9816>}
 dyn_hair_3_A_r {parent=dyn_hair_2_A_r, loc=<4.8748 0 0>, rot=<0 0 0 1>}
 base_hair_0_c_null {parent=head, type=Null, loc=<5.8985 -5.0186 0>, rot=<-0.0192 -0.7068 0.7068 0.0192>}
 base_hair_0_r_null {parent=head, type=Null, loc=<5.8985 -5.0186 5.6>, rot=<-0.0193 -0.7068 0.7068 0.0193>}
 base_hair_0_l_null {parent=head, type=Null, loc=<5.8985 -5.0186 -5.6>, rot=<0.0007 -0.7071 0.7071 -0.0007>}
 head_null {parent=spine_05, type=Null, loc=<20.6573 3.5489 0>, rot=<0 0 -0.036 -0.9994>}
 head_sp_null {parent=spine_05, type=Null, loc=<20.6573 3.5489 0>, rot=<0 0 -0.036 -0.9994>}
 clavicle_l {parent=spine_05, loc=<5.4831 -0.1053 -3.2946>, rot=<0.0465 -0.7289 -0.009 -0.6829>}
 upperarm_l {parent=clavicle_l, loc=<9.8658 -1.2136 -0.4105>, rot=<0.0171 -0.3487 -0.0265 -0.9367>}
 upperarm_twist_01_l {parent=upperarm_l, loc=<7.2135 -0.1134 0>, rot=<0 0 0 1>}
 clavicle_r {parent=spine_05, loc=<5.4831 -0.1053 3.2946>, rot=<0.7289 0.0465 -0.6829 0.009>}
 upperarm_r {parent=clavicle_r, loc=<-9.8658 1.2136 0.4105>, rot=<0.0164 -0.3245 -0.0269 -0.9454>}
 upperarm_twist_01_r {parent=upperarm_r, loc=<-7.2135 0.1134 0>, rot=<0 0 0 -1>}
 upperarm_l_null {parent=spine_05, type=Null, loc=<4.5519 -1.8922 -13.0371>, rot=<-0.0385 0.9118 0.0238 0.4081>}
 upperarm_r_null {parent=spine_05, type=Null, loc=<4.5519 -1.8922 13.0371>, rot=<-0.9118 -0.0385 0.4081 -0.0238>}
 dyn_hair_7_l_null {type=Null, loc=<5.6 -13.1597 124.0217>, rot=<-0.5148 -0.5148 0.4847 -0.4847>}
 dyn_hair_7_l_ctrl {controlType=EULER_TRANSFORM, parent=dyn_hair_7_l_null, type=Control, loc=<0 0 0>, rot=<0 0 0 1>}
 dyn_hair_6_l_null {type=Null, loc=<5.6 -13.4638 129.0586>, rot=<-0.5148 -0.5148 0.4847 -0.4847>}
 dyn_hair_6_l_ctrl {controlType=EULER_TRANSFORM, parent=dyn_hair_6_l_null, type=Control, loc=<0 0 0>, rot=<0 0 0 1>}
 dyn_hair_5_l_null {type=Null, loc=<5.6 -13.7682 134.1028>, rot=<-0.5148 -0.5148 0.4847 -0.4847>}
 dyn_hair_5_l_ctrl {controlType=EULER_TRANSFORM, parent=dyn_hair_5_l_null, type=Control, loc=<0 0 0>, rot=<0 0 0 1>}
 dyn_hair_4_l_null {type=Null, loc=<5.6 -13.2017 139.1172>, rot=<0.4711 0.4711 -0.5273 0.5273>}
 dyn_hair_4_l_ctrl {controlType=EULER_TRANSFORM, parent=dyn_hair_4_l_null, type=Control, loc=<0 0 0>, rot=<0 0 0 1>}
 dyn_hair_3_l_null {type=Null, loc=<5.6 -10.6353 143.8752>, rot=<0.3624 0.3624 -0.6072 0.6072>}
 dyn_hair_3_l_ctrl {controlType=EULER_TRANSFORM, parent=dyn_hair_3_l_null, type=Control, loc=<0 0 0>, rot=<0 0 0 1>}
 dyn_hair_2_l_null {type=Null, loc=<5.6 -8.3625 148.4053>, rot=<0.3713 0.3713 -0.6018 0.6018>}
 dyn_hair_2_l_ctrl {controlType=EULER_TRANSFORM, parent=dyn_hair_2_l_null, type=Control, loc=<0 0 0>, rot=<0 0 0 1>}
 dyn_hair_1_l_null {type=Null, loc=<5.6 -6.7546 153.204>, rot=<0.4154 0.4154 -0.5722 0.5722>}
 dyn_hair_1_l_ctrl {controlType=EULER_TRANSFORM, parent=dyn_hair_1_l_null, type=Control, loc=<0 0 0>, rot=<0 0 0 1>}
 dyn_hair_7_r_null {type=Null, loc=<-5.6 -13.1189 124.0076>, rot=<-0.5148 -0.5148 0.4847 -0.4847>}
 dyn_hair_7_r_ctrl {controlType=EULER_TRANSFORM, parent=dyn_hair_7_r_null, type=Control, loc=<0 0 0>, rot=<0 0 0 1>}
 dyn_hair_6_r_null {type=Null, loc=<-5.6 -13.4231 129.0444>, rot=<-0.5148 -0.5148 0.4847 -0.4847>}
 dyn_hair_6_r_ctrl {controlType=EULER_TRANSFORM, parent=dyn_hair_6_r_null, type=Control, loc=<0 0 0>, rot=<0 0 0 1>}
 dyn_hair_5_r_null {type=Null, loc=<-5.6 -13.7277 134.0882>, rot=<-0.5148 -0.5148 0.4847 -0.4847>}
 dyn_hair_5_r_ctrl {controlType=EULER_TRANSFORM, parent=dyn_hair_5_r_null, type=Control, loc=<0 0 0>, rot=<0 0 0 1>}
 dyn_hair_4_r_null {type=Null, loc=<-5.6 -13.1611 139.1028>, rot=<0.4711 0.4711 -0.5273 0.5273>}
 dyn_hair_4_r_ctrl {controlType=EULER_TRANSFORM, parent=dyn_hair_4_r_null, type=Control, loc=<0 0 0>, rot=<0 0 0 1>}
 dyn_hair_3_r_null {type=Null, loc=<-5.6 -10.5948 143.8611>, rot=<0.3624 0.3624 -0.6072 0.6072>}
 dyn_hair_3_r_ctrl {controlType=EULER_TRANSFORM, parent=dyn_hair_3_r_null, type=Control, loc=<0 0 0>, rot=<0 0 0 1>}
 dyn_hair_2_r_null {type=Null, loc=<-5.6 -8.3219 148.3906>, rot=<0.3713 0.3713 -0.6018 0.6018>}
 dyn_hair_2_r_ctrl {controlType=EULER_TRANSFORM, parent=dyn_hair_2_r_null, type=Control, loc=<0 0 0>, rot=<0 0 0 1>}
 dyn_hair_1_r_null {type=Null, loc=<-5.6 -6.7546 153.204>, rot=<-0.4154 -0.4154 0.5722 -0.5722>}
 dyn_hair_1_r_ctrl {controlType=EULER_TRANSFORM, parent=dyn_hair_1_r_null, type=Control, loc=<0 0 0>, rot=<0 0 0 1>}
 dyn_hair_7_c_null {type=Null, loc=<0 -13.1191 124.0072>, rot=<-0.5148 -0.5148 0.4847 -0.4847>}
 dyn_hair_7_c_ctrl {controlType=EULER_TRANSFORM, parent=dyn_hair_7_c_null, type=Control, loc=<0 0 0>, rot=<0 0 0 1>}
 dyn_hair_6_c_null {type=Null, loc=<0 -14.0481 128.8664>, rot=<-0.5148 -0.5148 0.4847 -0.4847>}
 dyn_hair_6_c_ctrl {controlType=EULER_TRANSFORM, parent=dyn_hair_6_c_null, type=Control, loc=<0 0 0>, rot=<0 0 0 1>}
 dyn_hair_5_c_null {type=Null, loc=<0 -13.7276 134.0883>, rot=<-0.5148 -0.5148 0.4847 -0.4847>}
 dyn_hair_5_c_ctrl {controlType=EULER_TRANSFORM, parent=dyn_hair_5_c_null, type=Control, loc=<0 0 0>, rot=<0 0 0 1>}
 dyn_hair_4_c_null {type=Null, loc=<0 -13.1611 139.1027>, rot=<0.4711 0.4711 -0.5273 0.5273>}
 dyn_hair_4_c_ctrl {controlType=EULER_TRANSFORM, parent=dyn_hair_4_c_null, type=Control, loc=<0 0 0>, rot=<0 0 0 1>}
 dyn_hair_3_c_null {type=Null, loc=<0 -10.5947 143.8607>, rot=<0.3624 0.3624 -0.6072 0.6072>}
 dyn_hair_3_c_ctrl {controlType=EULER_TRANSFORM, parent=dyn_hair_3_c_null, type=Control, loc=<0 0 0>, rot=<0 0 0 1>}
 dyn_hair_2_c_null {type=Null, loc=<0 -8.3219 148.3908>, rot=<0.3713 0.3713 -0.6018 0.6018>}
 dyn_hair_2_c_ctrl {controlType=EULER_TRANSFORM, parent=dyn_hair_2_c_null, type=Control, loc=<0 0 0>, rot=<0 0 0 1>}
 dyn_hair_1_c_null {type=Null, loc=<0 -6.7546 153.204>, rot=<-0.4154 -0.4154 0.5722 -0.5722>}
 dyn_hair_1_c_ctrl {controlType=EULER_TRANSFORM, parent=dyn_hair_1_c_null, type=Control, loc=<0 0 0>, rot=<0 0 0 1>}
 Delta_isHeadUp {type=Curve, loc=<0 0 0>, rot=<0 0 0 1>}
 Delta_isHeadUpExtreme {type=Curve, loc=<0 0 0>, rot=<0 0 0 1>}
 Delta_upperarmUp_L {type=Curve, loc=<0 0 0>, rot=<0 0 0 1>}
 Delta_upperarmUp_R {type=Curve, loc=<0 0 0>, rot=<0 0 0 1>}
 Delta_upperarmUpB_L {type=Curve, loc=<0 0 0>, rot=<0 0 0 1>}
 dyn_hair_0_c_null {type=Null, loc=<0 -6.7546 157.891>, rot=<-0.5154 -0.5154 0.4841 -0.4841>}
 dyn_hair_0_c_ctrl {controlType=EULER_TRANSFORM, parent=dyn_hair_0_c_null, type=Control, loc=<158.004 0 3.1542>, rot=<0.5154 0.5154 -0.4841 -0.4841>}
 dyn_hair_0_r_null {type=Null, loc=<-5.6 -6.7546 157.891>, rot=<-0.5154 -0.5154 0.4841 -0.4841>}
 dyn_hair_0_r_ctrl {controlType=EULER_TRANSFORM, parent=dyn_hair_0_r_null, type=Control, loc=<158.0039 5.6 3.1565>, rot=<0.5154 0.5154 -0.4841 -0.4841>}
 dyn_hair_0_l_null {type=Null, loc=<5.6 -6.7546 157.891>, rot=<-0.5016 -0.5016 0.4984 -0.4984>}
 dyn_hair_0_l_ctrl {controlType=EULER_TRANSFORM, parent=dyn_hair_0_l_null, type=Control, loc=<157.9304 -5.6 -5.7612>, rot=<0.5016 0.5016 -0.4984 -0.4984>}
]
variables [
 PubIsGunOut {type=bool, default=False}
 PubIsRigidBodiesEnabled {type=bool, default=False}
 PubIsWearingBackbling {type=bool, default=False}
 isCrouchWalking {type=double, default=0}
 isCrouching {type=bool, default=False}
 isPA_Swing {type=bool, default=False}
 isSkyFloating {type=bool, default=False}
 isTargeting {type=bool, default=False}
 isTargetingLerp {type=double, default=0}
]
