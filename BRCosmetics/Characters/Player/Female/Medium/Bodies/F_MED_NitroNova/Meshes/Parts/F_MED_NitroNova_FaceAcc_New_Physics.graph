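# Physics asset for the NitroNova female medium face-accessory mesh, covering three hair chains (center, left, right) and two shoulder hair strands.
# Capsule bodies cover upper spine, clavicles, upper arms, neck and head, with hair chains rooted at the head.
# Body and hair-root joints swing and twist freely, while lower hair segments have tight, asymmetric swing limits.
# Shoulder strands branch from the first left and right hair segments with narrow swing and restricted twist.
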PhysicsAsset /BRCosmetics/Characters/Player/Female/Medium/Bodies/F_MED_NitroNova/Meshes/Parts/F_MED_NitroNova_FaceAcc_New_Physics
bodies {
 base_hair_0_c [{shapeType=Capsule, length=0.5, radius=0.5, shapeName=base_hair_0_c_capsule}]
 base_hair_0_l [{shapeType=Capsule, length=0.5, radius=0.5, shapeName=base_hair_0_l_capsule}]
 base_hair_0_r [{shapeType=Capsule, length=0.5, radius=0.5, shapeName=dyn_hair_1_r_capsule}]
 clavicle_l [{
   shapeType=Capsule
   length=4.8744
   radius=6.9997
   shapeName=clavicle_l_capsule
   center=<7.554 -1.2208 -1.7517>
   rotation=<rot 80.0096 6.1869 2.5075>
  }]
 clavicle_r [{
   shapeType=Capsule
   length=4.8744
   radius=6.9997
   shapeName=clavicle_r_capsule
   center=<-7.7791 0.9777 1.5266>
   rotation=<rot 80 8.6968 179.9983>
  }]
 dyn_hairShoulder_l [{shapeType=Capsule, length=0.5, radius=0.5, shapeName=dyn_hairShoulder_l_capsule}]
 dyn_hairShoulder_l_end [{
   shapeType=Capsule
   length=4.1564
   radius=1.12
   shapeName=dyn_hairShoulder_l_end_capsule
   center=<1.3213 0.6761 1.8979>
   rotation=<rot -0.0001 0 19.6925>
  }]
 dyn_hairShoulder_r [{shapeType=Capsule, length=0.5, radius=0.5, shapeName=dyn_hairShoulder_r_capsule}]
 dyn_hairShoulder_r_end [
  {
   shapeType=Capsule
   length=1.4099
   radius=1.6296
   shapeName=dyn_hairShoulder_r_end_capsule
   center=<-0.6361 -3.9692 3.8407>
   rotation=<rot 19.5424 -4.0234 -11.8746>
  }
  {
   shapeType=Capsule
   length=4.6677
   radius=2
   shapeName=dyn_hairShoulder_r_end_capsule
   center=<0.5511 -2.0383 -2.0501>
   rotation=<rot -0.141 0.0278 -11.1815>
  }
  {
   shapeType=Capsule
   length=1.5512
   radius=1.6296
   shapeName=dyn_hairShoulder_r_end_capsule
   center=<-2.2055 -4.6166 8.1015>
   rotation=<rot 18.3543 -3.7603 -11.7892>
  }
 ]
 dyn_hair_0_c [{shapeType=Capsule, length=0.5, radius=0.5, shapeName=base_hair_0_c_capsule}]
 dyn_hair_0_l [{shapeType=Capsule, length=0.5, radius=0.5, shapeName=base_hair_0_l_capsule}]
 dyn_hair_0_r [{shapeType=Capsule, length=0.5, radius=0.5, shapeName=dyn_hair_1_r_capsule}]
 dyn_hair_1_c [{shapeType=Capsule, length=0.5, radius=0.5, shapeName=dyn_hair_1_c_capsule}]
 dyn_hair_1_l [{shapeType=Capsule, length=0.5, radius=0.5, shapeName=dyn_hair_1_l_capsule}]
 dyn_hair_1_r [{shapeType=Capsule, length=0.5, radius=0.5, shapeName=dyn_hair_1_r_capsule}]
 dyn_hair_2_c [{shapeType=Capsule, length=0.5, radius=0.5, shapeName=dyn_hair_2_c_capsule}]
 dyn_hair_2_l [{shapeType=Capsule, length=0.5, radius=0.5, shapeName=dyn_hair_2_l_capsule}]
 dyn_hair_2_r [{shapeType=Capsule, length=0.5, radius=0.5, shapeName=dyn_hair_2_r_capsule}]
 dyn_hair_3_c [{shapeType=Capsule, length=3.6884, radius=2.1517, shapeName=dyn_hair_3_c_capsule, center=<1 0 3.5145>, rotation=<rot 85 -180 -180>}]
 dyn_hair_3_l [
  {shapeType=Capsule, length=3.8739, radius=2.629, shapeName=dyn_hair_3_l_capsule, center=<1 0 0.81>, rotation=<rot 90 0 0>}
  {
   shapeType=Capsule
   length=3.8739
   radius=2.629
   shapeName=dyn_hair_3_l_capsule
   center=<0.8734 4.518 1.1766>
   rotation=<rot 85.0001 -180 -180>
  }
 ]
 dyn_hair_3_r [{shapeType=Capsule, length=4.6422, radius=2.629, shapeName=dyn_hair_3_r_capsule, center=<1 0 1.0988>, rotation=<rot 90 0 0>}]
 dyn_hair_4_c [{shapeType=Capsule, length=5.7059, radius=2.0357, shapeName=dyn_hair_4_c_capsule, center=<0.8786 0 3.6496>, rotation=<rot 85.0003 -180 -180>}]
 dyn_hair_4_l [
  {shapeType=Capsule, length=3.5313, radius=2.629, shapeName=dyn_hair_4_l_capsule, center=<1 0 1.0988>, rotation=<rot 90 0 0>}
  {
   shapeType=Capsule
   length=5.2814
   radius=2.0414
   shapeName=dyn_hair_4_l_capsule
   center=<1.702 2.8992 -0.6256>
   rotation=<rot 80.0001 -180 -180>
  }
 ]
 dyn_hair_4_r [
  {shapeType=Capsule, length=4.258, radius=2.629, shapeName=dyn_hair_4_r_capsule, center=<1 0 1.0988>, rotation=<rot 90 0 0>}
  {shapeType=Capsule, length=4.7836, radius=2.1034, shapeName=dyn_hair_4_r_capsule, center=<1 -4.4108 0.9334>, rotation=<rot 90 0 0>}
 ]
 dyn_hair_5_c [{shapeType=Capsule, length=1.2925, radius=2.8092, shapeName=dyn_hair_5_c_capsule, center=<1 0 2.7286>, rotation=<rot 90 0 0>}]
 dyn_hair_5_l [{shapeType=Capsule, length=2.3646, radius=2.629, shapeName=dyn_hair_5_l_capsule, center=<1 0 1.0988>, rotation=<rot 90 0 0>}]
 dyn_hair_5_r [{shapeType=Capsule, length=2.3646, radius=2.629, shapeName=dyn_hair_5_r_capsule, center=<1 0 1.0988>, rotation=<rot 90 0 0>}]
 dyn_hair_6_c [{shapeType=Capsule, length=1.9493, radius=2.6557, shapeName=dyn_hair_5_c_capsule, center=<1 0 1.5936>, rotation=<rot 90 0 0>}]
 dyn_hair_6_l [{shapeType=Capsule, length=0.7838, radius=1.994, shapeName=dyn_hair_6_l_capsule, center=<1 0 1.0988>, rotation=<rot 90 0 0>}]
 dyn_hair_6_r [{shapeType=Capsule, length=1.7005, radius=1.7076, shapeName=dyn_hair_6_r_capsule, center=<1 0 1.0988>, rotation=<rot 90 0 0>}]
 head [{shapeType=Capsule, length=3.5428, radius=8.1582, shapeName=head_capsule, center=<2.1229 1.9145 0>, rotation=<rot -90 0 0>}]
 neck_01 [
  {shapeType=Capsule, length=1.2686, radius=4.2417, shapeName=neck_01_capsule, center=<1.5184 0.4491 0>, rotation=<rot 90 0 0>}
  {shapeType=Capsule, length=1.3981, radius=4.6025, shapeName=neck_02_capsule, center=<2.0011 0.4576 0>, rotation=<rot 90 -2.5336 0>}
 ]
 spine_04 [{shapeType=Capsule, length=9.5975, radius=7.8232, shapeName=spine_04_capsule, center=<3.0071 -0.4749 0>}]
 spine_05 [
  {shapeType=Capsule, length=11.7587, radius=7.7566, shapeName=spine_05_capsule, center=<5.9865 -1.5661 0>, rotation=<rot 0 0 -180>}
  {shapeType=Capsule, length=10.0375, radius=7.8232, shapeName=spine_04_capsule, center=<-0.5854 -0.4677 0>, rotation=<rot 0 0 -180>}
 ]
 upperarm_l [{
   shapeType=Capsule
   length=14.2649
   radius=6.0972
   shapeName=upperarm_twist_01_l_capsule
   center=<8.6734 -0.1795 0.0589>
   rotation=<rot 88.9759 176.8886 0.0082>
  }]
 upperarm_r [{
   shapeType=Capsule
   length=14.2649
   radius=6.0972
   shapeName=upperarm_twist_01_r_capsule
   center=<-8.6734 0.1795 -0.0589>
   rotation=<rot 88.9759 176.8886 -179.9918>
  }]
}
constraints [
 {
  bone1Name=spine_05
  bone2Name=spine_04
  swing1LimitDegrees=45
  swing1Motion=Free
  swing2LimitDegrees=45
  swing2Motion=Free
  twistLimitDegrees=45
  twistMotion=Free
 }
 {
  bone1Name=upperarm_l
  bone2Name=clavicle_l
  swing1LimitDegrees=45
  swing1Motion=Free
  swing2LimitDegrees=45
  swing2Motion=Free
  twistLimitDegrees=45
  twistMotion=Free
 }
 {
  bone1Name=clavicle_l
  bone2Name=spine_05
  swing1LimitDegrees=45
  swing1Motion=Free
  swing2LimitDegrees=45
  swing2Motion=Free
  twistLimitDegrees=45
  twistMotion=Free
 }
 {
  bone1Name=clavicle_r
  bone2Name=spine_05
  swing1LimitDegrees=45
  swing1Motion=Free
  swing2LimitDegrees=45
  swing2Motion=Free
  twistLimitDegrees=45
  twistMotion=Free
 }
 {
  bone1Name=upperarm_r
  bone2Name=clavicle_r
  swing1LimitDegrees=45
  swing1Motion=Free
  swing2LimitDegrees=45
  swing2Motion=Free
  twistLimitDegrees=45
  twistMotion=Free
 }
 {
  bone1Name=neck_01
  bone2Name=spine_05
  swing1LimitDegrees=45
  swing1Motion=Free
  swing2LimitDegrees=45
  swing2Motion=Free
  twistLimitDegrees=45
  twistMotion=Free
 }
 {
  bone1Name=head
  bone2Name=neck_01
  swing1LimitDegrees=45
  swing1Motion=Free
  swing2LimitDegrees=45
  swing2Motion=Free
  twistLimitDegrees=45
  twistMotion=Free
 }
 {
  bone1Name=base_hair_0_c
  bone2Name=head
  swing1LimitDegrees=45
  swing1Motion=Free
  swing2LimitDegrees=45
  swing2Motion=Free
  twistLimitDegrees=45
  twistMotion=Free
 }
 {
  bone1Name=dyn_hair_0_c
  bone2Name=base_hair_0_c
  swing1LimitDegrees=45
  swing1Motion=Free
  swing2LimitDegrees=45
  swing2Motion=Free
  twistLimitDegrees=45
  twistMotion=Free
 }
 {
  bone1Name=dyn_hair_1_c
  bone2Name=dyn_hair_0_c
  swing1LimitDegrees=45
  swing1Motion=Free
  swing2LimitDegrees=45
  swing2Motion=Free
  twistLimitDegrees=45
  twistMotion=Free
 }
 {
  bone1Name=dyn_hair_2_c
  bone2Name=dyn_hair_1_c
  swing1LimitDegrees=45
  swing1Motion=Free
  swing2LimitDegrees=45
  swing2Motion=Free
  twistLimitDegrees=45
  twistMotion=Free
 }
 {bone1Name=dyn_hair_3_c, bone2Name=dyn_hair_2_c, swing1LimitDegrees=3, swing2LimitDegrees=30, twistLimitDegrees=10}
 {bone1Name=dyn_hair_4_c, bone2Name=dyn_hair_3_c, swing1LimitDegrees=10, swing2LimitDegrees=30, twistLimitDegrees=10}
 {bone1Name=dyn_hair_5_c, bone2Name=dyn_hair_4_c, swing1LimitDegrees=10, swing2LimitDegrees=30, twistLimitDegrees=10}
 {bone1Name=dyn_hair_6_c, bone2Name=dyn_hair_5_c, swing1LimitDegrees=10, swing2LimitDegrees=30, twistLimitDegrees=10}
 {
  bone1Name=dyn_hair_1_l
  bone2Name=dyn_hair_0_l
  swing1LimitDegrees=45
  swing1Motion=Free
  swing2LimitDegrees=45
  swing2Motion=Free
  twistLimitDegrees=45
  twistMotion=Free
 }
 {
  bone1Name=base_hair_0_l
  bone2Name=head
  swing1LimitDegrees=45
  swing1Motion=Free
  swing2LimitDegrees=45
  swing2Motion=Free
  twistLimitDegrees=45
  twistMotion=Free
 }
 {
  bone1Name=dyn_hair_0_l
  bone2Name=base_hair_0_l
  swing1LimitDegrees=45
  swing1Motion=Free
  swing2LimitDegrees=45
  swing2Motion=Free
  twistLimitDegrees=45
  twistMotion=Free
 }
 {bone1Name=dyn_hairShoulder_l, bone2Name=dyn_hair_1_l, swing1LimitDegrees=15, swing2LimitDegrees=5, twistLimitDegrees=45, twistMotion=Locked}
 {
  bone1Name=dyn_hairShoulder_l_end
  bone2Name=dyn_hairShoulder_l
  swing1LimitDegrees=15
  swing2LimitDegrees=5
  twistLimitDegrees=45
  twistMotion=Locked
 }
 {
  bone1Name=dyn_hair_2_l
  bone2Name=dyn_hair_1_l
  swing1LimitDegrees=45
  swing1Motion=Free
  swing2LimitDegrees=45
  swing2Motion=Free
  twistLimitDegrees=45
  twistMotion=Free
 }
 {bone1Name=dyn_hair_3_l, bone2Name=dyn_hair_2_l, swing1LimitDegrees=3, swing2LimitDegrees=30, twistLimitDegrees=10}
 {bone1Name=dyn_hair_4_l, bone2Name=dyn_hair_3_l, swing1LimitDegrees=10, swing2LimitDegrees=30, twistLimitDegrees=10}
 {bone1Name=dyn_hair_5_l, bone2Name=dyn_hair_4_l, swing1LimitDegrees=10, swing2LimitDegrees=30, twistLimitDegrees=10}
 {bone1Name=dyn_hair_6_l, bone2Name=dyn_hair_5_l, swing1LimitDegrees=10, swing2LimitDegrees=30, twistLimitDegrees=10}
 {
  bone1Name=base_hair_0_r
  bone2Name=head
  swing1LimitDegrees=45
  swing1Motion=Free
  swing2LimitDegrees=45
  swing2Motion=Free
  twistLimitDegrees=45
  twistMotion=Free
 }
 {
  bone1Name=dyn_hair_0_r
  bone2Name=base_hair_0_r
  swing1LimitDegrees=45
  swing1Motion=Free
  swing2LimitDegrees=45
  swing2Motion=Free
  twistLimitDegrees=45
  twistMotion=Free
 }
 {
  bone1Name=dyn_hair_1_r
  bone2Name=dyn_hair_0_r
  swing1LimitDegrees=45
  swing1Motion=Free
  swing2LimitDegrees=45
  swing2Motion=Free
  twistLimitDegrees=45
  twistMotion=Free
 }
 {bone1Name=dyn_hairShoulder_r, bone2Name=dyn_hair_1_r, swing1LimitDegrees=15, swing2LimitDegrees=5, twistLimitDegrees=0}
 {bone1Name=dyn_hairShoulder_r_end, bone2Name=dyn_hairShoulder_r, swing1LimitDegrees=15, swing2LimitDegrees=5, twistLimitDegrees=0}
 {
  bone1Name=dyn_hair_2_r
  bone2Name=dyn_hair_1_r
  swing1LimitDegrees=45
  swing1Motion=Free
  swing2LimitDegrees=45
  swing2Motion=Free
  twistLimitDegrees=45
  twistMotion=Free
 }
 {bone1Name=dyn_hair_4_r, bone2Name=dyn_hair_3_r, swing1LimitDegrees=10, swing2LimitDegrees=30, twistLimitDegrees=10}
 {bone1Name=dyn_hair_3_r, bone2Name=dyn_hair_2_r, swing1LimitDegrees=3, swing2LimitDegrees=30, twistLimitDegrees=10}
 {bone1Name=dyn_hair_5_r, bone2Name=dyn_hair_4_r, swing1LimitDegrees=10, swing2LimitDegrees=30, twistLimitDegrees=10}
 {bone1Name=dyn_hair_6_r, bone2Name=dyn_hair_5_r, swing1LimitDegrees=10, swing2LimitDegrees=30, twistLimitDegrees=10}
]
