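# Physics asset for the right-side front hair of the NitroNova female character's face accessory mesh.
# Capsules on the upper spine, neck, head, right clavicle and upper arm form body shapes for two hair chains, A and B, both rooted on the head.
# Body joints and both root joints rotate freely, while joints along the strands use tight swing and twist limits, with chain B stiffest at its base.
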PhysicsAsset /BRCosmetics/Characters/Player/Female/Medium/Bodies/F_MED_NitroNova/Meshes/Parts/F_MED_NitroNova_FaceAcc_Physics_FrontHair
bodies {
 base_hair_1_B_r [{shapeType=Capsule, length=5.2426, radius=1.9024, shapeName=base_hair_1_B_r_capsule, rotation=<rot 0 0 90>}]
 clavicle_r [{
   shapeType=Capsule
   length=6.9947
   radius=6.5004
   shapeName=clavicle_r_capsule
   center=<-7.1285 1.6733 1.5282>
   rotation=<rot 81.8628 1.3669 179.9993>
  }]
 dyn_hair_0_A_r [{
   shapeType=Capsule
   length=3.0876
   radius=1.9961
   shapeName=dyn_hair_0_A_r_capsule
   center=<4.5635 0.6968 2.1005>
   rotation=<rot -0.0795 94.9992 -89.0904>
  }]
 dyn_hair_1_A_r [{shapeType=Capsule, length=4.1851, radius=2.1276, shapeName=dyn_hair_1_A_r_capsule, rotation=<rot 0 90 90>}]
 dyn_hair_1_B_r [{shapeType=Capsule, length=5.2426, radius=1.9024, shapeName=dyn_hair_1_B_r_capsule, center=<3.6262 0.6681 0.1541>, rotation=<rot 0 0 90>}]
 dyn_hair_2_A_r [
  {shapeType=Capsule, length=4.5296, radius=2, shapeName=dyn_hair_2_A_r_capsule, center=<2.0973 -0.2243 0.4087>, rotation=<rot 0 89.9999 90>}
  {shapeType=Capsule, length=4.5296, radius=2, shapeName=dyn_hair_2_A_r_capsule, center=<2.4248 1.6564 0.4087>, rotation=<rot 0 69.9998 90>}
 ]
 dyn_hair_2_B_r [{shapeType=Capsule, length=10.2724, radius=1.8574, shapeName=dyn_hair_2_B_r_capsule, center=<2.8383 0.5979 0.1497>, rotation=<rot 0 0 90>}]
 dyn_hair_3_A_r [
  {
   shapeType=Capsule
   length=3.7387
   radius=1.3153
   shapeName=dyn_hair_3_A_r_capsule
   center=<0.1534 1.5787 3.1084>
   rotation=<rot -16.1702 24.2742 144.904>
  }
  {
   shapeType=Capsule
   length=5.2729
   radius=1.6813
   shapeName=dyn_hair_3_A_r_capsule
   center=<2.6294 -1.0087 1.5634>
   rotation=<rot -43.3789 22.4969 -32.4551>
  }
 ]
 dyn_hair_3_B_r [{
   shapeType=Capsule
   length=4.1868
   radius=2
   shapeName=dyn_hair_3_B_r_capsule
   center=<2.6298 0.3227 -0.8095>
   rotation=<rot -60.7372 180 179.9999>
  }]
 head [{shapeType=Capsule, length=5.0217, radius=7.6582, shapeName=head_capsule, center=<2.1229 1.9145 0>, rotation=<rot -90 0 0>}]
 neck_01 [{shapeType=Capsule, length=0.3883, radius=5.122, shapeName=neck_01_capsule, center=<1.5184 0.4491 0>, rotation=<rot 90 0 0>}]
 neck_02 [{shapeType=Capsule, length=1.3981, radius=4.6025, shapeName=neck_02_capsule, center=<2.0011 0.4576 0>, rotation=<rot 90 -2.5336 0>}]
 spine_04 [{shapeType=Capsule, length=9.3881, radius=8.0327, shapeName=spine_04_capsule, center=<1.8264 3.4119 0>}]
 spine_05 [
  {shapeType=Capsule, length=6.6388, radius=7.7566, shapeName=spine_05_capsule, center=<3.3752 -1.1381 0>}
  {shapeType=Capsule, length=11.4142, radius=6.861, shapeName=head_capsule, center=<20.0808 -4.2253 8.1216>, rotation=<rot 90 24.9996 0>}
 ]
 upperarm_twist_01_r [{
   shapeType=Capsule
   length=20.4167
   radius=4.9935
   shapeName=upperarm_twist_01_r_capsule
   center=<-6.281 0.6789 1.0104>
   rotation=<rot 88.9759 176.8886 -179.9918>
  }]
}
constraints [
 {
  bone1Name=neck_01
  bone2Name=spine_05
  swing1LimitDegrees=45
  swing1Motion=Free
  swing2LimitDegrees=45
  swing2Motion=Free
  twistLimitDegrees=45
  twistMotion=Free
 }
 {
  bone1Name=neck_02
  bone2Name=neck_01
  swing1LimitDegrees=45
  swing1Motion=Free
  swing2LimitDegrees=45
  swing2Motion=Free
  twistLimitDegrees=45
  twistMotion=Free
 }
 {
  bone1Name=head
  bone2Name=neck_02
  swing1LimitDegrees=45
  swing1Motion=Free
  swing2LimitDegrees=45
  swing2Motion=Free
  twistLimitDegrees=45
  twistMotion=Free
 }
 {
  bone1Name=spine_05
  bone2Name=spine_04
  swing1LimitDegrees=45
  swing1Motion=Free
  swing2LimitDegrees=45
  swing2Motion=Free
  twistLimitDegrees=45
  twistMotion=Free
 }
 {
  bone1Name=clavicle_r
  bone2Name=spine_05
  swing1LimitDegrees=45
  swing1Motion=Free
  swing2LimitDegrees=45
  swing2Motion=Free
  twistLimitDegrees=45
  twistMotion=Free
 }
 {
  bone1Name=upperarm_twist_01_r
  bone2Name=clavicle_r
  swing1LimitDegrees=45
  swing1Motion=Free
  swing2LimitDegrees=45
  swing2Motion=Free
  twistLimitDegrees=45
  twistMotion=Free
 }
 {bone1Name=dyn_hair_2_A_r, bone2Name=dyn_hair_1_A_r, swing1LimitDegrees=20, swing2LimitDegrees=30, twistLimitDegrees=10}
 {bone1Name=dyn_hair_3_A_r, bone2Name=dyn_hair_2_A_r, swing1LimitDegrees=25, swing2LimitDegrees=30, twistLimitDegrees=10}
 {
  bone1Name=base_hair_1_B_r
  bone2Name=head
  swing1LimitDegrees=45
  swing1Motion=Free
  swing2LimitDegrees=45
  swing2Motion=Free
  twistLimitDegrees=45
  twistMotion=Free
 }
 {bone1Name=dyn_hair_1_B_r, bone2Name=base_hair_1_B_r, swing1LimitDegrees=5, swing2LimitDegrees=10, twistLimitDegrees=10}
 {bone1Name=dyn_hair_2_B_r, bone2Name=dyn_hair_1_B_r, swing1LimitDegrees=10, swing2LimitDegrees=20, twistLimitDegrees=10}
 {bone1Name=dyn_hair_3_B_r, bone2Name=dyn_hair_2_B_r, swing1LimitDegrees=10, swing2LimitDegrees=20, twistLimitDegrees=10}
 {bone1Name=dyn_hair_1_A_r, bone2Name=dyn_hair_0_A_r, swing1LimitDegrees=35, swing2LimitDegrees=25, twistLimitDegrees=15}
 {
  bone1Name=dyn_hair_0_A_r
  bone2Name=head
  swing1LimitDegrees=45
  swing1Motion=Free
  swing2LimitDegrees=45
  swing2Motion=Free
  twistLimitDegrees=45
  twistMotion=Free
 }
]
