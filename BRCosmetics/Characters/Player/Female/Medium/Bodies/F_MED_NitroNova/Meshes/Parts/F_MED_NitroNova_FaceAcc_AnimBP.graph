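# Post-process style hair animation Blueprint for the NitroNova face accessory, adding secondary motion to hair and braids.
# Parent and BodyMesh poses feed a Control Rig, then a bool blend of RigidBody or bone scaling, then AnimDynamics and front-hair RigidBody.
# Braid bones follow hair bones through weighted constraints, then get AnimDynamics and final corrective offsets.
# Events update gravity, front-end/match sim space, and weapon, glider, backbling and pickaxe state.
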AnimBlueprint /BRCosmetics/Characters/Player/Female/Medium/Bodies/F_MED_NitroNova/Meshes/Parts/F_MED_NitroNova_FaceAcc_AnimBP
target_skeleton=&/F_MED_NitroNova_FaceAcc_Skeleton
graphs [
 AnimGraph {nodes [
   Output Pose(Result=Use cached pose 'OutputCache'()) {layerGroup=None}
   n1 = Get BodyMesh() -> (BodyMesh:Skeletal Mesh Component Object Reference)
   n2 = Control Rig!(
    Source=Layered blend per bone(BasePose=Copy Pose From Mesh() {bCopyCurves=true, bUseAttachedParent=true}, BlendPoses_0=Copy Pose From Mesh(SourceMeshComponent=@n1)) {layerSetup [{branchFilters [{boneName=upperarm_l}, {boneName=upperarm_r}]}]}
    Alpha
    PubIsRigidBodiesEnabled:bool=false
    PubIsWearingBackbling:bool=false
    isCrouching:bool=True
    isTargeting:bool=false
    isSkyFloating:bool=false
    PubIsGunOut:bool=false
    isPA_Swing:bool=false
   ) -> (Pose) {controlRigAssetReference {blueprintRigClass=&/F_MED_NitroNova_FaceAcc_CtrlRig.F_MED_NitroNova_FaceAcc_CtrlRig_C}}
   Save cached pose 'CR_Input'(Pose=@n2) {cacheName=CR_Input}
   n3 = Local To Component(LocalPose=Use cached pose 'PostBraidsCache'())
   n4 = "Transform (Modify) Bone - Bone: dyn_hair_4_r"!(
    ComponentPose="Transform (Modify) Bone - Bone: dyn_hair_1_c"!(
     ComponentPose="Transform (Modify) Bone - Bone: dyn_hair_5_c"!(ComponentPose=@n3, bAlphaBoolEnabled:bool=True, Rotation="0,0,15") -> (Pose) {alphaInputType=Bool, rotationMode=BMM_Additive, boneToModify {boneName=dyn_hair_5_c}, rotation {roll=15}}
     bAlphaBoolEnabled:bool=True
     Rotation="0,0,15"
    ) -> (Pose) {alphaInputType=Bool, rotationMode=BMM_Additive, boneToModify {boneName=dyn_hair_1_c}, rotation {roll=15}}
    bAlphaBoolEnabled:bool=True
    Translation="0,0.287,4"
   ) -> (Pose) {alphaInputType=Bool, translationMode=BMM_Additive, boneToModify {boneName=dyn_hair_4_r}, translation {y=0.287, z=4}}
   Save cached pose 'OutputCache'(Pose=Component To Local(ComponentPose=@n4)) {cacheName=OutputCache}
   n5 = "Constraint - dyn_braid_1_l (PPPPPPPP)"!(
    ComponentPose=Local To Component(LocalPose=Use cached pose 'CR_BackHairDyn'())
    Alpha=1
    ConstraintWeights_0=0.6235
    ConstraintWeights_1:float=0.2702
    ConstraintWeights_2:float=0.0469
    ConstraintWeights_3:float=0.0398
    ConstraintWeights_4:float=0.0085
    ConstraintWeights_5:float=0.0074
    ConstraintWeights_6:float=0.0021
    ConstraintWeights_7:float=0.0016
   ) -> (Pose) {
    boneToModify {boneName=dyn_braid_1_l}
    constraintSetup [
     {targetBone {boneName=head}}
     {targetBone {boneName=dyn_hair_1_l}}
     {targetBone {boneName=dyn_hair_1_c}}
     {targetBone {boneName=dyn_hair_2_l}}
     {targetBone {boneName=dyn_hair_2_c}}
     {targetBone {boneName=dyn_hair_3_l}}
     {targetBone {boneName=dyn_hair_3_c}}
     {targetBone {boneName=dyn_hair_1_r}}
    ]
    constraintWeights [1,1,1,1,1,1,1,1]
   }
   n6 = "Constraint - dyn_braid_2_l (PPPPPPPP)"!(
    ComponentPose=@n5
    Alpha=1
    ConstraintWeights_0=0.6676
    ConstraintWeights_1:float=0.1893
    ConstraintWeights_2:float=0.0625
    ConstraintWeights_3:float=0.0434
    ConstraintWeights_4:float=0.0175
    ConstraintWeights_5:float=0.0093
    ConstraintWeights_6:float=0.007
    ConstraintWeights_7:float=0.0034
   ) -> (Pose) {
    boneToModify {boneName=dyn_braid_2_l}
    constraintSetup [
     {targetBone {boneName=dyn_hairShoulder_l}}
     {targetBone {boneName=dyn_hairShoulder_l_end}}
     {targetBone {boneName=dyn_hair_1_l}}
     {targetBone {boneName=dyn_hair_2_l}}
     {targetBone {boneName=dyn_hair_3_l}}
     {targetBone {boneName=dyn_hair_1_c}}
     {targetBone {boneName=dyn_hair_2_c}}
     {targetBone {boneName=dyn_hair_4_l}}
    ]
    constraintWeights [1,1,1,1,1,1,1,1]
   }
   n7 = "Constraint - dyn_braid_3_l (PPPPPPPP)"!(
    ComponentPose=@n6
    Alpha=1
    ConstraintWeights_0=0.3464
    ConstraintWeights_1:float=0.2188
    ConstraintWeights_2:float=0.206
    ConstraintWeights_3:float=0.089
    ConstraintWeights_4:float=0.0827
    ConstraintWeights_5:float=0.0251
    ConstraintWeights_6:float=0.0169
    ConstraintWeights_7:float=0.0151
   ) -> (Pose) {
    boneToModify {boneName=dyn_braid_3_l}
    constraintSetup [
     {targetBone {boneName=dyn_hairShoulder_l}}
     {targetBone {boneName=dyn_hair_3_l}}
     {targetBone {boneName=dyn_hair_2_l}}
     {targetBone {boneName=dyn_hair_4_l}}
     {targetBone {boneName=dyn_hair_1_l}}
     {targetBone {boneName=dyn_hair_5_l}}
     {targetBone {boneName=dyn_hair_3_c}}
     {targetBone {boneName=dyn_hairShoulder_l_end}}
    ]
    constraintWeights [1,1,1,1,1,1,1,1]
   }
   n8 = "Constraint - dyn_braid_1_c (PPPPPPPP)"!(
    ComponentPose=@n7
    Alpha=1
    ConstraintWeights_0=0.6255
    ConstraintWeights_1:float=0.2682
    ConstraintWeights_2:float=0.0475
    ConstraintWeights_3:float=0.0394
    ConstraintWeights_4:float=0.0085
    ConstraintWeights_5:float=0.0073
    ConstraintWeights_6:float=0.0021
    ConstraintWeights_7:float=0.0016
   ) -> (Pose) {
    boneToModify {boneName=dyn_braid_1_c}
    constraintSetup [
     {targetBone {boneName=head}}
     {targetBone {boneName=dyn_hair_1_l}}
     {targetBone {boneName=dyn_hair_1_c}}
     {targetBone {boneName=dyn_hair_2_l}}
     {targetBone {boneName=dyn_hair_2_c}}
     {targetBone {boneName=dyn_hair_3_l}}
     {targetBone {boneName=dyn_hair_3_c}}
     {targetBone {boneName=dyn_hair_1_r}}
    ]
    constraintWeights [1,1,1,1,1,1,1,1]
   }
   n9 = "Constraint - dyn_braid_2_c (PPPPPPPP)"!(
    ComponentPose=@n8
    Alpha=1
    ConstraintWeights_0=0.3872
    ConstraintWeights_1:float=0.2495
    ConstraintWeights_2:float=0.1815
    ConstraintWeights_3:float=0.0702
    ConstraintWeights_4:float=0.0474
    ConstraintWeights_5:float=0.0372
    ConstraintWeights_6:float=0.0148
    ConstraintWeights_7:float=0.0121
   ) -> (Pose) {
    boneToModify {boneName=dyn_braid_2_c}
    constraintSetup [
     {targetBone {boneName=dyn_hairShoulder_l}}
     {targetBone {boneName=dyn_hair_1_l}}
     {targetBone {boneName=dyn_hair_2_l}}
     {targetBone {boneName=dyn_hair_3_l}}
     {targetBone {boneName=dyn_hair_1_c}}
     {targetBone {boneName=dyn_hair_2_c}}
     {targetBone {boneName=dyn_hair_3_c}}
     {targetBone {boneName=dyn_hair_4_l}}
    ]
    constraintWeights [1,1,1,1,1,1,1,1]
   }
   n10 = "Constraint - dyn_braid_3_c (PPPPPPPP)"!(
    ComponentPose=@n9
    Alpha=1
    ConstraintWeights_0=0.3262
    ConstraintWeights_1:float=0.3203
    ConstraintWeights_2:float=0.1155
    ConstraintWeights_3:float=0.1128
    ConstraintWeights_4:float=0.0354
    ConstraintWeights_5:float=0.0314
    ConstraintWeights_6:float=0.0299
    ConstraintWeights_7:float=0.0285
   ) -> (Pose) {
    boneToModify {boneName=dyn_braid_3_c}
    constraintSetup [
     {targetBone {boneName=dyn_hair_3_l}}
     {targetBone {boneName=dyn_hair_2_l}}
     {targetBone {boneName=dyn_hair_4_l}}
     {targetBone {boneName=dyn_hair_1_l}}
     {targetBone {boneName=dyn_hair_3_c}}
     {targetBone {boneName=dyn_hair_5_l}}
     {targetBone {boneName=dyn_hairShoulder_l}}
     {targetBone {boneName=dyn_hair_2_c}}
    ]
    constraintWeights [1,1,1,1,1,1,1,1]
   }
   n11 = "Constraint - dyn_braid_4_c (PPPPPPPP)"!(
    ComponentPose=@n10
    Alpha=1
    ConstraintWeights_0=0.2796
    ConstraintWeights_1:float=0.2766
    ConstraintWeights_2:float=0.1197
    ConstraintWeights_3:float=0.1023
    ConstraintWeights_4:float=0.0788
    ConstraintWeights_5:float=0.0764
    ConstraintWeights_6:float=0.0344
    ConstraintWeights_7:float=0.0322
   ) -> (Pose) {
    boneToModify {boneName=dyn_braid_4_c}
    constraintSetup [
     {targetBone {boneName=dyn_hair_4_l}}
     {targetBone {boneName=dyn_hair_3_l}}
     {targetBone {boneName=dyn_hair_5_l}}
     {targetBone {boneName=dyn_hair_2_l}}
     {targetBone {boneName=dyn_hair_3_c}}
     {targetBone {boneName=dyn_hair_4_c}}
     {targetBone {boneName=dyn_hair_6_l}}
     {targetBone {boneName=dyn_hair_5_c}}
    ]
    constraintWeights [1,1,1,1,1,1,1,1]
   }
   n12 = "Constraint - dyn_braid_1_r (PPPPPPPP)"!(
    ComponentPose=@n11
    Alpha=1
    ConstraintWeights_0=0.5269
    ConstraintWeights_1:float=0.321
    ConstraintWeights_2:float=0.0854
    ConstraintWeights_3:float=0.0416
    ConstraintWeights_4:float=0.0124
    ConstraintWeights_5:float=0.0074
    ConstraintWeights_6:float=0.0027
    ConstraintWeights_7:float=0.0025
   ) -> (Pose) {
    boneToModify {boneName=dyn_braid_1_r}
    constraintSetup [
     {targetBone {boneName=head}}
     {targetBone {boneName=dyn_hair_1_l}}
     {targetBone {boneName=dyn_hair_1_c}}
     {targetBone {boneName=dyn_hair_2_l}}
     {targetBone {boneName=dyn_hair_2_c}}
     {targetBone {boneName=dyn_hair_3_l}}
     {targetBone {boneName=dyn_hair_1_r}}
     {targetBone {boneName=dyn_hair_3_c}}
    ]
    constraintWeights [1,1,1,1,1,1,1,1]
   }
   n13 = "Constraint - dyn_braid_2_r (PPPPPPPP)"!(
    ComponentPose=@n12
    Alpha=1
    ConstraintWeights_0=0.3641
    ConstraintWeights_1:float=0.224
    ConstraintWeights_2:float=0.1547
    ConstraintWeights_3:float=0.0951
    ConstraintWeights_4:float=0.0716
    ConstraintWeights_5:float=0.0508
    ConstraintWeights_6:float=0.0299
    ConstraintWeights_7:float=0.0098
   ) -> (Pose) {
    boneToModify {boneName=dyn_braid_2_r}
    constraintSetup [
     {targetBone {boneName=dyn_hair_1_l}}
     {targetBone {boneName=dyn_hair_2_l}}
     {targetBone {boneName=dyn_hair_1_c}}
     {targetBone {boneName=dyn_hair_2_c}}
     {targetBone {boneName=dyn_hair_3_l}}
     {targetBone {boneName=head}}
     {targetBone {boneName=dyn_hair_3_c}}
     {targetBone {boneName=dyn_hair_4_l}}
    ]
    constraintWeights [1,1,1,1,1,1,1,1]
   }
   n14 = "Constraint - dyn_braid_3_r (PPPPPPPP)"!(
    ComponentPose=@n13
    Alpha=1
    ConstraintWeights_0=0.3622
    ConstraintWeights_1:float=0.2881
    ConstraintWeights_2:float=0.1161
    ConstraintWeights_3:float=0.0712
    ConstraintWeights_4:float=0.0615
    ConstraintWeights_5:float=0.0585
    ConstraintWeights_6:float=0.0306
    ConstraintWeights_7:float=0.0119
   ) -> (Pose) {
    boneToModify {boneName=dyn_braid_3_r}
    constraintSetup [
     {targetBone {boneName=dyn_hair_2_l}}
     {targetBone {boneName=dyn_hair_3_l}}
     {targetBone {boneName=dyn_hair_1_l}}
     {targetBone {boneName=dyn_hair_2_c}}
     {targetBone {boneName=dyn_hair_4_l}}
     {targetBone {boneName=dyn_hair_3_c}}
     {targetBone {boneName=dyn_hair_1_c}}
     {targetBone {boneName=dyn_hair_4_c}}
    ]
    constraintWeights [1,1,1,1,1,1,1,1]
   }
   Save cached pose 'CR_BraidConstraint'(Pose=Component To Local(ComponentPose=@n14)) {cacheName=CR_BraidConstraint}
   n15 = RigidBody!(
    ComponentPose=Local To Component(LocalPose=Use cached pose 'CR_Input'())
    Alpha
    OverrideWorldGravity:Vector="0,0,0"
    SimSpaceSettings:Sim Space Settings Structure=Get Sim Space RigB() -> (Sim Space RigB:Sim Space Settings Structure)
   ) -> (Pose) {
    bOverrideWorldGravity=true
    overridePhysicsAsset=&/F_MED_NitroNova_FaceAcc_New_Physics
    simulationSpace=BaseBoneSpace
    alphaScaleBiasClamp {bInterpResult=true, interpSpeedDecreasing=3, interpSpeedIncreasing=3}
    baseBoneRef {boneName=spine_05}
    componentAppliedLinearAccClamp=<3000 3000 3000>
    simSpaceSettings {
     maxAngularAcceleration=100
     maxAngularVelocity=3
     maxLinearAcceleration=800
     maxLinearVelocity=400
     worldAlpha=1
     externalLinearDragV=<1 1 1>
    }
   }
   n16 = Component To Local(ComponentPose=@n15)
   n17 = "Transform (Modify) Bone - Bone: dyn_hair_0_c"!(ComponentPose=Local To Component(LocalPose=Use cached pose 'CR_Input'()), bAlphaBoolEnabled:bool=True, Scale="1,1,0.75") -> (Pose) {
    alphaInputType=Bool
    rotationSpace=BCS_BoneSpace
    scaleMode=BMM_Additive
    scaleSpace=BCS_BoneSpace
    translationSpace=BCS_BoneSpace
    alphaScaleBiasClamp {bInterpResult=true}
    boneToModify {boneName=dyn_hair_0_c}
    scale {z=0.75}
   }
   n18 = "Transform (Modify) Bone - Bone: dyn_hair_0_r"!(ComponentPose=@n17, bAlphaBoolEnabled:bool=True, Scale="1,1,0.75") -> (Pose) {
    alphaInputType=Bool
    rotationSpace=BCS_BoneSpace
    scaleMode=BMM_Additive
    scaleSpace=BCS_BoneSpace
    translationSpace=BCS_BoneSpace
    alphaScaleBiasClamp {bInterpResult=true}
    boneToModify {boneName=dyn_hair_0_r}
    scale {z=0.75}
   }
   n19 = "Transform (Modify) Bone - Bone: dyn_hair_0_l"!(ComponentPose=@n18, bAlphaBoolEnabled:bool=True, Scale="1,1,0.75") -> (Pose) {
    alphaInputType=Bool
    rotationSpace=BCS_BoneSpace
    scaleMode=BMM_Additive
    scaleSpace=BCS_BoneSpace
    translationSpace=BCS_BoneSpace
    alphaScaleBiasClamp {bInterpResult=true}
    boneToModify {boneName=dyn_hair_0_l}
    scale {z=0.75}
   }
   n20 = Component To Local(ComponentPose=@n19)
   n21 = Anim Dynamics - Bone: base_hair_0_A_r!(
    ComponentPose=Local To Component(LocalPose=Blend Poses by bool(BlendPose_0=@n20, BlendPose_1=@n16) {blendType=Linear})
    bAlphaBoolEnabled:bool=True
    GravityOverride:Vector="0,0,-980"
    ExternalForce="500,-500,0"
   ) -> (Pose) {
    alphaInputType=Bool
    angularDampingOverride=1
    angularSpringConstant=500
    bAngularSpring=true
    bOverrideAngularDamping=true
    bUseGravityOverride=true
    gravityScale=0.35
    simulationSpace=RootRelative
    alphaBoolBlend {blendInTime=0.2, blendOutTime=0.2}
    boundBone {boneName=base_hair_0_A_r}
    componentLinearAccScale=<1 1 1>
    componentLinearVelScale=<1 1 1>
    gravityOverride {z=-980}
    physicsBodyDefinitions [{
      boundBone {boneName=base_hair_0_A_r}
      boxExtents=<35 35 35>
      constraintSetup {angularLimitsMax=<0 -10 25>, angularLimitsMin=<0 -25 -30>, angularTarget=<1 0 0>}
      localJointOffset=<3 0 0>
     }]
    simSpaceSettings {maxAngularAcceleration=30, maxAngularVelocity=5, simSpaceAngularAlpha=1}
   }
   n22 = Anim Dynamics - Bone: base_hair_1_B_r!(ComponentPose=@n21, bAlphaBoolEnabled:bool=True, GravityOverride:Vector="0,0,-980", ExternalForce="500,-500,0") -> (Pose) {
    alphaInputType=Bool
    angularDampingOverride=1
    angularSpringConstant=500
    bAngularSpring=true
    bOverrideAngularDamping=true
    bUseGravityOverride=true
    gravityScale=0.35
    simulationSpace=RootRelative
    alphaBoolBlend {blendInTime=0.2, blendOutTime=0.2}
    boundBone {boneName=base_hair_1_B_r}
    componentLinearAccScale=<1 1 1>
    componentLinearVelScale=<1 1 1>
    gravityOverride {z=-980}
    physicsBodyDefinitions [{
      boundBone {boneName=base_hair_1_B_r}
      boxExtents=<45 45 45>
      constraintSetup {angularLimitsMax=<0 -5 25>, angularLimitsMin=<0 -10 10>, angularTarget=<1 0 0>}
      localJointOffset=<1 0 0>
     }]
    simSpaceSettings {maxAngularAcceleration=30, maxAngularVelocity=5, simSpaceAngularAlpha=1}
   }
   n23 = RigidBody(ComponentPose=@n22, OverrideWorldGravity="0,0,0", SimSpaceSettings=Get Sim Space RigB()) {
    bOverrideWorldGravity=true
    overridePhysicsAsset=&/F_MED_NitroNova_FaceAcc_Physics_FrontHair
    simulationSpace=BaseBoneSpace
    alphaScaleBiasClamp {bInterpResult=true, interpSpeedDecreasing=3, interpSpeedIncreasing=3}
    baseBoneRef {boneName=spine_05}
    componentAppliedLinearAccClamp=<3000 3000 3000>
    simSpaceSettings {
     maxAngularAcceleration=100
     maxAngularVelocity=3
     maxLinearAcceleration=800
     maxLinearVelocity=400
     worldAlpha=1
     externalLinearDragV=<1 1 1>
    }
   }
   Save cached pose 'CR_BackHairDyn'(Pose=Component To Local(ComponentPose=@n23)) {cacheName=CR_BackHairDyn}
   n24 = Anim Dynamics - Bone: dyn_braid_3_l!(
    ComponentPose=Local To Component(LocalPose=Use cached pose 'CR_BraidConstraint'())
    Alpha=1
    GravityOverride:Vector="0,0,-980"
    ExternalForce="0,0,0"
   ) -> (Pose) {
    angularDampingOverride=1
    angularSpringConstant=1000
    bAngularSpring=true
    bOverrideAngularDamping=true
    bUseGravityOverride=true
    gravityScale=0.35
    simulationSpace=RootRelative
    boundBone {boneName=dyn_braid_3_l}
    componentAppliedLinearAccClamp=<3000 3000 3000>
    componentLinearAccScale=<0.4 0.4 0.4>
    componentLinearVelScale=<0.4 0.4 0.4>
    gravityOverride {z=-980}
    physicsBodyDefinitions [{
      boundBone {boneName=dyn_braid_3_l}
      boxExtents=<20 20 20>
      constraintSetup {angularLimitsMax=<0 0 5>, angularLimitsMin=<0 -15 -5>, angularTarget=<1 0 0>}
      localJointOffset=<3 0 0>
     }]
   }
   n25 = Anim Dynamics - Bone: dyn_braid_4_l!(ComponentPose=@n24, Alpha=1, GravityOverride:Vector="0,0,-980", ExternalForce="0,0,0") -> (Pose) {
    angularDampingOverride=1
    angularSpringConstant=1000
    bAngularSpring=true
    bOverrideAngularDamping=true
    bUseGravityOverride=true
    gravityScale=0.35
    simulationSpace=RootRelative
    boundBone {boneName=dyn_braid_4_l}
    componentAppliedLinearAccClamp=<3000 3000 3000>
    componentLinearAccScale=<0.4 0.4 0.4>
    componentLinearVelScale=<0.4 0.4 0.4>
    gravityOverride {z=-980}
    physicsBodyDefinitions [{
      boundBone {boneName=dyn_braid_4_l}
      boxExtents=<20 20 20>
      constraintSetup {angularLimitsMax=<0 0 5>, angularLimitsMin=<0 -15 -5>, angularTarget=<1 0 0>}
      localJointOffset=<3 0 0>
     }]
   }
   n26 = Anim Dynamics - Bone: dyn_braid_3_c!(ComponentPose=@n25, Alpha=1, GravityOverride:Vector="0,0,-980", ExternalForce="0,0,0") -> (Pose) {
    angularDampingOverride=1
    angularSpringConstant=1000
    bAngularSpring=true
    bOverrideAngularDamping=true
    bUseGravityOverride=true
    gravityScale=0.35
    simulationSpace=RootRelative
    boundBone {boneName=dyn_braid_3_c}
    componentAppliedLinearAccClamp=<3000 3000 3000>
    componentLinearAccScale=<0.4 0.4 0.4>
    componentLinearVelScale=<0.4 0.4 0.4>
    gravityOverride {z=-980}
    physicsBodyDefinitions [{
      boundBone {boneName=dyn_braid_3_c}
      boxExtents=<20 20 20>
      constraintSetup {angularLimitsMax=<0 0 5>, angularLimitsMin=<0 -15 -5>, angularTarget=<1 0 0>}
      localJointOffset=<3 0 0>
     }]
   }
   n27 = Anim Dynamics - Bone: dyn_braid_4_c!(ComponentPose=@n26, Alpha=1, GravityOverride:Vector="0,0,-980", ExternalForce="0,0,0") -> (Pose) {
    angularDampingOverride=1
    angularSpringConstant=1000
    bAngularSpring=true
    bOverrideAngularDamping=true
    bUseGravityOverride=true
    gravityScale=0.35
    simulationSpace=RootRelative
    boundBone {boneName=dyn_braid_4_c}
    componentAppliedLinearAccClamp=<3000 3000 3000>
    componentLinearAccScale=<0.4 0.4 0.4>
    componentLinearVelScale=<0.4 0.4 0.4>
    gravityOverride {z=-980}
    physicsBodyDefinitions [{
      boundBone {boneName=dyn_braid_4_c}
      boxExtents=<20 20 20>
      constraintSetup {angularLimitsMax=<0 0 5>, angularLimitsMin=<0 -15 -5>, angularTarget=<1 0 0>}
      localJointOffset=<3 0 0>
     }]
   }
   n28 = Anim Dynamics - Bone: dyn_braid_5_c!(ComponentPose=@n27, Alpha=1, GravityOverride:Vector="0,0,-980", ExternalForce="0,0,0") -> (Pose) {
    angularDampingOverride=1
    angularSpringConstant=1000
    bAngularSpring=true
    bOverrideAngularDamping=true
    bUseGravityOverride=true
    gravityScale=0.35
    simulationSpace=RootRelative
    boundBone {boneName=dyn_braid_5_c}
    componentAppliedLinearAccClamp=<3000 3000 3000>
    componentLinearAccScale=<0.4 0.4 0.4>
    componentLinearVelScale=<0.4 0.4 0.4>
    gravityOverride {z=-980}
    physicsBodyDefinitions [{
      boundBone {boneName=dyn_braid_5_c}
      boxExtents=<20 20 20>
      constraintSetup {angularLimitsMax=<0 0 5>, angularLimitsMin=<0 -15 -5>, angularTarget=<1 0 0>}
      localJointOffset=<3 0 0>
     }]
   }
   n29 = Anim Dynamics - Bone: dyn_braid_3_r!(ComponentPose=@n28, Alpha=1, GravityOverride:Vector="0,0,-980", ExternalForce="0,0,0") -> (Pose) {
    angularDampingOverride=1
    angularSpringConstant=1000
    bAngularSpring=true
    bOverrideAngularDamping=true
    bUseGravityOverride=true
    gravityScale=0.35
    simulationSpace=RootRelative
    boundBone {boneName=dyn_braid_3_r}
    componentAppliedLinearAccClamp=<3000 3000 3000>
    componentLinearAccScale=<0.4 0.4 0.4>
    componentLinearVelScale=<0.4 0.4 0.4>
    gravityOverride {z=-980}
    physicsBodyDefinitions [{
      boundBone {boneName=dyn_braid_3_r}
      boxExtents=<20 20 20>
      constraintSetup {angularLimitsMax=<0 0 5>, angularLimitsMin=<0 -15 -5>, angularTarget=<1 0 0>}
      localJointOffset=<3 0 0>
     }]
   }
   n30 = Anim Dynamics - Bone: dyn_braid_4_r!(ComponentPose=@n29, Alpha=1, GravityOverride:Vector="0,0,-980", ExternalForce="0,0,0") -> (Pose) {
    angularDampingOverride=1
    angularSpringConstant=1000
    bAngularSpring=true
    bOverrideAngularDamping=true
    bUseGravityOverride=true
    gravityScale=0.35
    simulationSpace=RootRelative
    boundBone {boneName=dyn_braid_4_r}
    componentAppliedLinearAccClamp=<3000 3000 3000>
    componentLinearAccScale=<0.4 0.4 0.4>
    componentLinearVelScale=<0.4 0.4 0.4>
    gravityOverride {z=-980}
    physicsBodyDefinitions [{
      boundBone {boneName=dyn_braid_4_r}
      boxExtents=<20 20 20>
      constraintSetup {angularLimitsMax=<0 0 5>, angularLimitsMin=<0 -15 -5>, angularTarget=<1 0 0>}
      localJointOffset=<3 0 0>
     }]
   }
   Save cached pose 'PostBraidsCache'(Pose=Component To Local(ComponentPose=@n30)) {cacheName=PostBraidsCache}
  ]}
 EventGraph {nodes [
   Event BlueprintUpdateAnimation()
   > n1 = "Gravity Override State Function (Parameter Asset) (Nativized)"(
    CharacterPartAnimInstance=Self-Reference()
    GameTimeInSeconds=GetGameTimeInSeconds()
    Parameters=&/F_MED_NitroNova_FaceAcc_Parameters
   )
   > Set GravityOverride(GravityOverride=@n1)
   Event OnParachuteOpened()
   > Switch on EGliderType(Selection=Get PlayerGliderType() -> (PlayerGliderType=HangGlider))
   > Set isHangGliding(isHangGliding:bool=true) -> (Output_Get:bool)
   n2 = Event OnCharacterCustomizationLoaded()
   > Set BodyMesh(BodyMesh=@n2.CustomBaseMesh)
   n3 = CheckForBackbling()
   > Set isWearingBackbling(isWearingBackbling=OR Boolean!(A=@n3, B=@n3.bIsBackblingACape, C:bool=@n3.bIsBackblingLong) -> (ReturnValue))
   > n4 = Branch(Condition=Get bIsInFrontEnd())
   > Set Sim Space RigB(Sim Space RigB:Sim Space Settings Structure=Get Sim Space FE() -> (Sim Space FE:Sim Space Settings Structure)) -> (Output_Get:Sim Space Settings Structure)
   Set Sim Space RigB(execute=@n4.else, Sim Space RigB=Get Sim Space BR() -> (Sim Space BR:Sim Space Settings Structure))
   Event OnWeaponChanged()
   > Set isGunOut(isGunOut:bool=true) -> (Output_Get:bool)
   > n5 = Switch on EFortWeaponCoreAnimation!(execute, Selection:EFortWeaponCoreAnimation Enum=Get WeaponCoreAnim!() -> (WeaponCoreAnim)) -> (
    Melee:Exec
    Pistol:Exec
    Shotgun:Exec
    PaperBlueprint:Exec
    Rifle:Exec
    MeleeOneHand:Exec
    MachinePistol:Exec
    RocketLauncher:Exec
    GrenadeLauncher:Exec
    GoingCommando:Exec
    AssaultRifle:Exec
    TacticalShotgun:Exec
    SniperRifle:Exec
    TrapPlacement:Exec
    ShoulderLauncher:Exec
    AbilityDecoTool:Exec
    Crossbow:Exec
    C4:Exec
    RemoteControl:Exec
    DualWield:Exec
    AR_BullPup:Exec
    AR_ForwardGrip:Exec
    MedPackPaddles:Exec
    SMG_P90:Exec
    AR_DrumGun:Exec
    Consumable_Small:Exec
    Consumable_Large:Exec
    Balloon:Exec
    MountedTurret:Exec
    CreativeTool:Exec
    ExplosiveBow:Exec
    AshtonIndigo:Exec
    AshtonChicago:Exec
    MeleeDualWield:Exec
    Unarmed:Exec
   )
   Set isGunOut(execute=[@n5.DualWield, @n5.Melee, @n5.MeleeDualWield, @n5.MeleeOneHand, @n5.Unarmed], isGunOut=false)
   Event OnParachuteClosed()
   > Set isHangGliding(isHangGliding=false)
   n6 = Event OnBeginPickaxeSwing() {is=title K2Node_Event}
   Branch(execute=[@n6.then, @n7.then], Condition=Get bIsPlayingMeleeAnim())
   > Set alphaHairRigidBody(alphaHairRigidBody:double=0.6) -> (Output_Get:double)
   n7 = Event OnEndPickaxeSwing() {is=title K2Node_Event}
  ]}
 sinAlphaBasic {nodes [
   n1 = sinAlphaBasic!() -> (then, Offset:double=~, Speed:double=~, Amplitude:double=~) {is=title K2Node_FunctionEntry}
   n2 = float + float(A=GetGameTimeInSeconds(), B=@n1)
   n3 = MultiplyByPi!(Value:double="Sin (Radians)"(A=float * float(A=@n2, B=@n1.Speed))) -> (ReturnValue:double)
   > Return Node!(execute, sinOutput:Rotator=MakeRotator(Pitch=float * float(A=@n3, B=@n1.Amplitude)))
  ]}
]
variables [
 alphaHairRigidBody {type=number, default=1}
 BodyMesh {type=SkeletalMeshComponent}
 DeltaArmUpFix {type=number}
 FakeWind {type=number}
 gameTime {type=number}
 Gravity Override Params {
  type=FortGravityOverrideParameterSet
  default=(Skydive_GoingUp=(GravityOverride=(WindFrequency=(X=1.500000,Y=7.000000,Z=8.000000),WindAmplitude=(X=200.000000,Y=-500.000000,Z=400.000000),GravityOverride=(X=0.000000,Y=0.000000,Z=-980.000000),JointName="root")),Skydive_FloatingDown=(GravityOverride=(WindFrequency=(X=1.500000,Y=7.000000,Z=8.000000),WindAmplitude=(X=200.000000,Y=-500.000000,Z=400.000000),WindOffset=(X=0.000000,Y=-600.000000,Z=400.000000),JointName="root")),Skydive_DivingDown=(GravityOverride=(WindFrequency=(X=1.500000,Y=12.000000,Z=10.000000),WindAmplitude=(X=50.000000,Y=300.000000,Z=100.000000),JointName="pelvis")),Skydive_Parachuting=(GravityOverride=(WindFrequency=(X=1.500000,Y=7.000000,Z=8.000000),WindAmplitude=(X=200.000000,Y=10.000000,Z=50.000000),GravityOverride=(X=0.000000,Y=0.000000,Z=-1200.000000),JointName="root")),OnGround_Standing=(GravityOverride=(GravityOverride=(X=0.000000,Y=0.000000,Z=-980.000000),JointName="root")),OnGround_Moving=(GravityOverride=(GravityOverride=(X=0.000000,Y=0.000000,Z=-980.000000),JointName="root")),OnGround_Moving_HighRotationRate=(GravityOverride=(GravityOverride=(X=0.000000,Y=0.000000,Z=-980.000000),JointName="root")),OnGround_Crouch_Moving=(GravityOverride=(GravityOverride=(X=0.000000,Y=0.000000,Z=-980.000000),JointName="root")),Emote_or_Melee=(GravityOverride=(GravityOverride=(X=0.000000,Y=0.000000,Z=-980.000000),JointName="root")),Swimming=(GravityOverride=(GravityOverride=(X=-800.000000,Y=0.000000,Z=0.000000),JointName="pelvis")),Front_End=(GravityOverride=(GravityOverride=(X=0.000000,Y=0.000000,Z=-980.000000),JointName="root")),DBNO=(GravityOverride=(GravityOverride=(X=-980.000000,Y=400.000000,Z=0.000000),JointName="pelvis")),DBNO_Carried=(GravityOverride=(GravityOverride=(X=-980.000000,Y=500.000000,Z=-500.000000),JointName="pelvis")),OnGround_Falling=(GravityOverride=(GravityOverride=(X=0.000000,Y=0.000000,Z=-980.000000),JointName="root")),Ziplining=(GravityOverride=(GravityOverride=(X=0.000000,Y=0.000000,Z=-980.000000),JointName="root")),Skydive_Surfing=(GravityOverride=(WindFrequency=(X=1.500000,Y=7.000000,Z=8.000000),WindAmplitude=(X=200.000000,Y=-500.000000,Z=400.000000),GravityOverride=(X=500.000000,Y=-300.000000,Z=-980.000000),JointName="root")),Riding_Standing=(GravityOverride=(GravityOverride=(X=0.000000,Y=-350.000000,Z=-980.000000),JointName="root")),Riding_Moving=(GravityOverride=(GravityOverride=(X=0.000000,Y=-450.000000,Z=-450.000000),JointName="root")),Riding_Moving_Targeting=(GravityOverride=(GravityOverride=(X=0.000000,Y=-200.000000,Z=-500.000000),JointName="root")),Tactical_Sprint=(GravityOverride=(GravityOverride=(X=0.000000,Y=-450.000000,Z=-500.000000),JointName="root")),Rail_Grinding=(GravityOverride=(GravityOverride=(X=0.000000,Y=0.000000,Z=-980.000000),JointName="root")),)
 }
 GravityOverride {type=Vector}
 hairBaseOverlap rotator {type=Rotator}
 hairBaseOverlap vector {type=Vector}
 isGunOut {type=boolean}
 isHangGliding {type=boolean}
 IsTacticalJump {type=boolean}
 isWearingBackbling {type=boolean}
 Sim Space BR {
  type=SimSpaceSettings
  default {
   dampingAlpha=1
   maxAngularAcceleration=100
   maxAngularVelocity=3
   maxLinearAcceleration=800
   maxLinearVelocity=400
   velocityScaleZ=1
   worldAlpha=1
   externalLinearDragV=<1 1 1>
  }
 }
 Sim Space FE {
  type=SimSpaceSettings
  default {
   dampingAlpha=1
   maxAngularAcceleration=1000
   maxAngularVelocity=3
   maxLinearAcceleration=1000
   maxLinearVelocity=500
   velocityScaleZ=1.5
   worldAlpha=1
   externalLinearDragV=<0.1 0.1 0.2>
  }
 }
 Sim Space RigB {
  type=SimSpaceSettings
  default {
   dampingAlpha=1
   maxAngularAcceleration=1000
   maxAngularVelocity=3
   maxLinearAcceleration=800
   maxLinearVelocity=400
   velocityScaleZ=1
   worldAlpha=1
   externalLinearDragV=<1 1 1>
  }
 }
 Sin_Rot_01 {type=Rotator}
 Sin_Rot_02 {type=Rotator}
 Sin_Rot_03 {type=Rotator}
]
