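# Physics asset for the back hair on the NitroNova female medium character's face accessory mesh, using one capsule per bone.
# Left, center and right hair chains hang from the head, with two-bone shoulder strands branching off the left and right chains.
# Constraints leave rotation free at roots and body bones, limit it down the chains, and lock twist on upper segments and shoulder strands.
# Head, neck, upper spine, clavicle and upper-arm capsules provide body shapes around the hair.
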PhysicsAsset /BRCosmetics/Characters/Player/Female/Medium/Bodies/F_MED_NitroNova/Meshes/Parts/F_MED_NitroNova_FaceAcc_Physics_BackHair
bodies {
 base_hair_0_A_r [{shapeType=Capsule, length=0.505, radius=0.505, shapeName=base_hair_0_A_r_capsule, rotation=<rot 0 -90 0>}]
 base_hair_0_c [{shapeType=Capsule, length=0.505, radius=0.505, shapeName=base_hair_0_c_capsule, rotation=<rot 0 -90 0>}]
 base_hair_0_l [{shapeType=Capsule, length=0.505, radius=0.505, shapeName=base_hair_0_l_capsule, rotation=<rot 0 -90 0>}]
 clavicle_l [{shapeType=Capsule, length=5.3737, radius=6.5004, shapeName=clavicle_l_capsule, center=<7.6164 -0.5504 -1.1173>, rotation=<rot 90 8.6984 0>}]
 clavicle_r [{shapeType=Capsule, length=5.3737, radius=6.5004, shapeName=clavicle_r_capsule, center=<-7.6164 0.5504 1.1173>, rotation=<rot 90 -171.3016 0>}]
 dyn_hairShoulder_l [{
   shapeType=Capsule
   length=8.2274
   radius=1.3506
   shapeName=dyn_hairShoulder_l_capsule
   center=<1.5407 0.976 2.7825>
   rotation=<rot 18.3557 -88.9392 -4.1223>
  }]
 dyn_hairShoulder_l_end [{
   shapeType=Capsule
   length=4.1564
   radius=1.12
   shapeName=dyn_hairShoulder_l_end_capsule
   center=<1.3213 0.6761 1.8979>
   rotation=<rot -0.0001 0 19.6925>
  }]
 dyn_hairShoulder_r [{
   shapeType=Capsule
   length=12.1718
   radius=1.9507
   shapeName=dyn_hairShoulder_r_capsule
   center=<-0.5274 -3.1894 5.3233>
   rotation=<rot -9.0867 -89.6127 -8.1948>
  }]
 dyn_hairShoulder_r_end [{
   shapeType=Capsule
   length=5.5512
   radius=1.6296
   shapeName=dyn_hairShoulder_r_end_capsule
   center=<1.3213 -2.4577 2.1999>
   rotation=<rot 6.321 -1.2547 -11.2508>
  }]
 dyn_hair_0_c [{
   shapeType=Capsule
   length=0.505
   radius=0.505
   shapeName=dyn_hair_0_c_capsule
   center=<3.8825 0.7568 4.2102>
   rotation=<rot 0 -3.0631 -63.7977>
  }]
 dyn_hair_0_l [{
   shapeType=Capsule
   length=2.9238
   radius=1.5179
   shapeName=dyn_hair_0_l_capsule
   center=<3.7748 -1.0336 2.1428>
   rotation=<rot 0 -8.2289 -62.6175>
  }]
 dyn_hair_0_r [{
   shapeType=Capsule
   length=4.6692
   radius=1.8885
   shapeName=dyn_hair_0_r_capsule
   center=<3.8942 0.3571 2.5099>
   rotation=<rot 0 -178.9247 -57.1963>
  }]
 dyn_hair_1_c [{shapeType=Capsule, length=2, radius=2, shapeName=dyn_hair_1_c_capsule, center=<2 0 0>, rotation=<rot 0 0 90>}]
 dyn_hair_1_l [{shapeType=Capsule, length=2, radius=2, shapeName=dyn_hair_1_l_capsule, center=<2 0 0>, rotation=<rot 0 0 90>}]
 dyn_hair_1_r [{shapeType=Capsule, length=2, radius=2, shapeName=dyn_hair_1_r_capsule, center=<2 0 0>, rotation=<rot 0 0 90>}]
 dyn_hair_2_c [{shapeType=Capsule, length=2, radius=2, shapeName=dyn_hair_2_c_capsule, center=<2 0 0>, rotation=<rot 0 0 90>}]
 dyn_hair_2_l [{shapeType=Capsule, length=4.548, radius=2, shapeName=dyn_hair_2_l_capsule, center=<2 0.8246 0>, rotation=<rot 0 0 -90>}]
 dyn_hair_2_r [{shapeType=Capsule, length=4.548, radius=2, shapeName=dyn_hair_2_r_capsule, center=<2 -0.8246 0>, rotation=<rot 0 0 90>}]
 dyn_hair_3_c [{shapeType=Capsule, length=2.3335, radius=1.6665, shapeName=dyn_hair_3_c_capsule, center=<2 0 0.2399>, rotation=<rot 0 0 90>}]
 dyn_hair_3_l [{shapeType=Capsule, length=4.5029, radius=2, shapeName=dyn_hair_3_l_capsule, center=<2 0.5761 -0.3451>, rotation=<rot 0 0 90>}]
 dyn_hair_3_r [{shapeType=Capsule, length=5.2712, radius=2, shapeName=dyn_hair_3_r_capsule, center=<2 -0.811 -0.3451>, rotation=<rot 0 0 -90>}]
 dyn_hair_4_c [{shapeType=Capsule, length=2, radius=2, shapeName=dyn_hair_4_c_capsule, center=<2 0 0.5848>, rotation=<rot 0 0 90>}]
 dyn_hair_4_l [{shapeType=Capsule, length=4.1603, radius=2, shapeName=dyn_hair_4_l_capsule, center=<2 0.2493 0>, rotation=<rot 0 0 90>}]
 dyn_hair_4_r [{shapeType=Capsule, length=4.887, radius=2, shapeName=dyn_hair_4_r_capsule, center=<2 -0.6647 0>, rotation=<rot 0 0 -90>}]
 dyn_hair_5_c [{shapeType=Capsule, length=2, radius=2, shapeName=dyn_hair_5_c_capsule, center=<2 0 0.0433>, rotation=<rot 0 0 90>}]
 dyn_hair_5_l [{shapeType=Capsule, length=2.9936, radius=2, shapeName=dyn_hair_5_l_capsule, center=<2 -0.4298 0.0433>, rotation=<rot 0 0 90>}]
 dyn_hair_5_r [{shapeType=Capsule, length=2.9936, radius=2, shapeName=dyn_hair_5_r_capsule, center=<2 0.4298 0.0433>, rotation=<rot 0 0 -90>}]
 dyn_hair_6_c [{shapeType=Capsule, length=2, radius=2, shapeName=dyn_hair_5_c_capsule, center=<2 0 0.0433>, rotation=<rot 0 0 90>}]
 dyn_hair_6_l [{shapeType=Capsule, length=1.4128, radius=1.365, shapeName=dyn_hair_6_l_capsule, center=<1.3599 -2.6841 0.0038>, rotation=<rot 90 0 0>}]
 dyn_hair_6_r [{
   shapeType=Capsule
   length=2.3295
   radius=1.0786
   shapeName=dyn_hair_6_r_capsule
   center=<2 1.1592 0.4243>
   rotation=<rot 81.1265 -180 -165.5027>
  }]
 head [{shapeType=Capsule, length=3.5428, radius=7.6582, shapeName=head_capsule, center=<2.1229 1.9145 0>, rotation=<rot -90 0 0>}]
 neck_01 [{shapeType=Capsule, length=1.2686, radius=4.2417, shapeName=neck_01_capsule, center=<1.5184 0.4491 0>, rotation=<rot 90 0 0>}]
 neck_02 [{shapeType=Capsule, length=1.3981, radius=4.6025, shapeName=neck_02_capsule, center=<2.0011 0.4576 0>, rotation=<rot 90 -2.5336 0>}]
 spine_04 [{shapeType=Capsule, length=9.3881, radius=8.0327, shapeName=spine_04_capsule, center=<3.5807 -0.7155 0>}]
 spine_05 [{shapeType=Capsule, length=6.6388, radius=7.7566, shapeName=spine_05_capsule, center=<4.2132 -0.8971 0>}]
 upperarm_twist_01_l [{
   shapeType=Capsule
   length=14.2649
   radius=6.0972
   shapeName=upperarm_twist_01_l_capsule
   center=<5.3806 0 0>
   rotation=<rot 88.9759 176.8886 0.0082>
  }]
 upperarm_twist_01_r [{
   shapeType=Capsule
   length=14.2649
   radius=6.0972
   shapeName=upperarm_twist_01_r_capsule
   center=<-5.3806 0 0>
   rotation=<rot 88.9759 176.8886 -179.9918>
  }]
}
constraints [
 {bone1Name=dyn_hair_3_c, bone2Name=dyn_hair_2_c, swing1LimitDegrees=3, swing2LimitDegrees=30, twistLimitDegrees=10}
 {bone1Name=dyn_hair_4_c, bone2Name=dyn_hair_3_c, swing1LimitDegrees=10, swing2LimitDegrees=30, twistLimitDegrees=10}
 {bone1Name=dyn_hair_5_c, bone2Name=dyn_hair_4_c, swing1LimitDegrees=10, swing2LimitDegrees=30, twistLimitDegrees=10}
 {bone1Name=dyn_hair_6_c, bone2Name=dyn_hair_5_c, swing1LimitDegrees=10, swing2LimitDegrees=30, twistLimitDegrees=10}
 {bone1Name=dyn_hair_3_l, bone2Name=dyn_hair_2_l, swing1LimitDegrees=3, swing2LimitDegrees=30, twistLimitDegrees=10}
 {bone1Name=dyn_hair_4_l, bone2Name=dyn_hair_3_l, swing1LimitDegrees=10, swing2LimitDegrees=30, twistLimitDegrees=10}
 {bone1Name=dyn_hair_5_l, bone2Name=dyn_hair_4_l, swing1LimitDegrees=10, swing2LimitDegrees=30, twistLimitDegrees=10}
 {bone1Name=dyn_hair_6_l, bone2Name=dyn_hair_5_l, swing1LimitDegrees=10, swing2LimitDegrees=30, twistLimitDegrees=10}
 {bone1Name=dyn_hair_3_r, bone2Name=dyn_hair_2_r, swing1LimitDegrees=3, swing2LimitDegrees=30, twistLimitDegrees=10}
 {bone1Name=dyn_hair_4_r, bone2Name=dyn_hair_3_r, swing1LimitDegrees=10, swing2LimitDegrees=30, twistLimitDegrees=10}
 {bone1Name=dyn_hair_5_r, bone2Name=dyn_hair_4_r, swing1LimitDegrees=10, swing2LimitDegrees=30, twistLimitDegrees=10}
 {bone1Name=dyn_hair_6_r, bone2Name=dyn_hair_5_r, swing1LimitDegrees=10, swing2LimitDegrees=30, twistLimitDegrees=10}
 {bone1Name=dyn_hair_2_c, bone2Name=dyn_hair_1_c, swing1LimitDegrees=3, swing2LimitDegrees=45, twistLimitDegrees=45, twistMotion=Locked}
 {bone1Name=dyn_hair_2_l, bone2Name=dyn_hair_1_l, swing1LimitDegrees=3, swing2LimitDegrees=45, twistLimitDegrees=45, twistMotion=Locked}
 {bone1Name=dyn_hair_2_r, bone2Name=dyn_hair_1_r, swing1LimitDegrees=3, swing2LimitDegrees=45, twistLimitDegrees=45, twistMotion=Locked}
 {
  bone1Name=neck_01
  bone2Name=spine_05
  swing1LimitDegrees=45
  swing1Motion=Free
  swing2LimitDegrees=45
  swing2Motion=Free
  twistLimitDegrees=45
  twistMotion=Free
 }
 {
  bone1Name=neck_02
  bone2Name=neck_01
  swing1LimitDegrees=45
  swing1Motion=Free
  swing2LimitDegrees=45
  swing2Motion=Free
  twistLimitDegrees=45
  twistMotion=Free
 }
 {
  bone1Name=head
  bone2Name=neck_02
  swing1LimitDegrees=45
  swing1Motion=Free
  swing2LimitDegrees=45
  swing2Motion=Free
  twistLimitDegrees=45
  twistMotion=Free
 }
 {
  bone1Name=spine_05
  bone2Name=spine_04
  swing1LimitDegrees=45
  swing1Motion=Free
  swing2LimitDegrees=45
  swing2Motion=Free
  twistLimitDegrees=45
  twistMotion=Free
 }
 {
  bone1Name=clavicle_l
  bone2Name=spine_05
  swing1LimitDegrees=45
  swing1Motion=Free
  swing2LimitDegrees=45
  swing2Motion=Free
  twistLimitDegrees=45
  twistMotion=Free
 }
 {
  bone1Name=clavicle_r
  bone2Name=spine_05
  swing1LimitDegrees=45
  swing1Motion=Free
  swing2LimitDegrees=45
  swing2Motion=Free
  twistLimitDegrees=45
  twistMotion=Free
 }
 {
  bone1Name=upperarm_twist_01_l
  bone2Name=clavicle_l
  swing1LimitDegrees=45
  swing1Motion=Free
  swing2LimitDegrees=45
  swing2Motion=Free
  twistLimitDegrees=45
  twistMotion=Free
 }
 {
  bone1Name=upperarm_twist_01_r
  bone2Name=clavicle_r
  swing1LimitDegrees=45
  swing1Motion=Free
  swing2LimitDegrees=45
  swing2Motion=Free
  twistLimitDegrees=45
  twistMotion=Free
 }
 {bone1Name=dyn_hairShoulder_l, bone2Name=dyn_hair_1_l, swing1LimitDegrees=15, swing2LimitDegrees=5, twistLimitDegrees=45, twistMotion=Locked}
 {
  bone1Name=dyn_hairShoulder_l_end
  bone2Name=dyn_hairShoulder_l
  swing1LimitDegrees=15
  swing2LimitDegrees=5
  twistLimitDegrees=45
  twistMotion=Locked
 }
 {bone1Name=dyn_hairShoulder_r, bone2Name=dyn_hair_1_r, swing1LimitDegrees=30, swing2LimitDegrees=5, twistLimitDegrees=45, twistMotion=Locked}
 {
  bone1Name=dyn_hairShoulder_r_end
  bone2Name=dyn_hairShoulder_r
  swing1LimitDegrees=15
  swing2LimitDegrees=5
  twistLimitDegrees=45
  twistMotion=Locked
 }
 {
  bone1Name=dyn_hair_1_c
  bone2Name=dyn_hair_0_c
  swing1LimitDegrees=45
  swing1Motion=Free
  swing2LimitDegrees=45
  swing2Motion=Free
  twistLimitDegrees=45
  twistMotion=Free
 }
 {
  bone1Name=base_hair_0_c
  bone2Name=head
  swing1LimitDegrees=45
  swing1Motion=Free
  swing2LimitDegrees=45
  swing2Motion=Free
  twistLimitDegrees=45
  twistMotion=Free
 }
 {
  bone1Name=dyn_hair_1_l
  bone2Name=dyn_hair_0_l
  swing1LimitDegrees=45
  swing1Motion=Free
  swing2LimitDegrees=45
  swing2Motion=Free
  twistLimitDegrees=45
  twistMotion=Free
 }
 {
  bone1Name=base_hair_0_l
  bone2Name=head
  swing1LimitDegrees=45
  swing1Motion=Free
  swing2LimitDegrees=45
  swing2Motion=Free
  twistLimitDegrees=45
  twistMotion=Free
 }
 {
  bone1Name=base_hair_0_A_r
  bone2Name=head
  swing1LimitDegrees=45
  swing1Motion=Free
  swing2LimitDegrees=45
  swing2Motion=Free
  twistLimitDegrees=45
  twistMotion=Free
 }
 {
  bone1Name=dyn_hair_0_l
  bone2Name=base_hair_0_l
  swing1LimitDegrees=45
  swing1Motion=Free
  swing2LimitDegrees=45
  swing2Motion=Free
  twistLimitDegrees=45
  twistMotion=Free
 }
 {
  bone1Name=dyn_hair_1_r
  bone2Name=dyn_hair_0_r
  swing1LimitDegrees=45
  swing1Motion=Free
  swing2LimitDegrees=45
  swing2Motion=Free
  twistLimitDegrees=45
  twistMotion=Free
 }
 {
  bone1Name=dyn_hair_0_r
  bone2Name=head
  swing1LimitDegrees=45
  swing1Motion=Free
  swing2LimitDegrees=45
  swing2Motion=Free
  twistLimitDegrees=45
  twistMotion=Free
 }
 {
  bone1Name=dyn_hair_0_c
  bone2Name=base_hair_0_c
  swing1LimitDegrees=45
  swing1Motion=Free
  swing2LimitDegrees=45
  swing2Motion=Free
  twistLimitDegrees=45
  twistMotion=Free
 }
]
